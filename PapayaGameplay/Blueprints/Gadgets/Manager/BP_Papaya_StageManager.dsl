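; Central controller for the Papaya concert stage, exposing command events for screens, lights, lasers, hologram dancers, disco balls, effects, post process and time of day.
; Most events set StageSetup variables, while timelines drive fades and whole-stage looks apply through UpdateEverything.
; BeginPlay starts a color-scroll timer, binds audio spectral analysis for cosmetics and restores screen transforms.
; Tagged mesh net events drive local player gameplay effects, supporting join-in-progress.
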(event OnMeshNetEvent(GameplayEffectEventResponder)
  (params (EventTag :type Gameplay%20Tag%20Structure) (bPersistentOnOff :type Boolean) (EventType :type EEventResponderEventType%20Enum :default "OneShotEvent") (EventParameters :type String))
  (bind _returnvalue (Utilities|Enum|NotEqual(Enum) EventType "PersistentEventEnd"))
  (bind _localgameplayeffectvolume (Variables|GameplayEffects|GetLocalGameplayEffectVolume))
  ;[0.04979 1 0] Mesh Event Responder binding to properly support server JIP behavior, TODO make these the only way to drive these and trigger on clients removing the need for the rep_notify bool behavior
  (Utilities|IsValid (Variables|GameplayEffects|GetLocalGameplayEffectVolume)
    (:Is%20Valid
      _)
    (:Is%20Not%20Valid
      (bind _outactors (Actor|GetAllActorsOfClasswithTag "/PapayaGameplay/Blueprints/BGA_Papaya_GameplayEffect_LocalFX.BGA_Papaya_GameplayEffect_LocalFX_C" "LocalFXVolume"))
      (if (== (Utilities|Array|Length _outactors) 1)
        (Variables|GameplayEffects|SetLocalGameplayEffectVolume (Utilities|Array|Get(acopy) _outactors)))
      (Development|PrintString "Papaya Couldn't find BGA_Papaya_GameplayEffect_LocalFX volume to make mesh commands work properly")))
  (switch Utilities|FlowControl|Switch|SwitchonGameplayTag EventTag
    (:Papaya.PlayerBouncy
      (Class|BGAPapayaGameplayEffectLocalFX|HandleMeshGameplayEffect _localgameplayeffectvolume (Variables|GameplayEffects|GetBounce%20Gameplay%20Effect) _returnvalue)
      (Class|BGAPapayaGameplayEffectLocalFX|HandleMeshGameplayEffect _localgameplayeffectvolume (Variables|GameplayEffects|GetLowG%20Gameplay%20Effect) _returnvalue))
    (:Papaya.PlayerFireJump
      (Class|BGAPapayaGameplayEffectLocalFX|HandleMeshGameplayEffect _localgameplayeffectvolume (Variables|GameplayEffects|GetFireJump%20Gameplay%20Effect) _returnvalue)
      (Class|BGAPapayaGameplayEffectLocalFX|HandleMeshGameplayEffect _localgameplayeffectvolume (Variables|GameplayEffects|GetLowG%20Gameplay%20Effect) _returnvalue))
    (:Papaya.PlayerBubbles
      (Class|BGAPapayaGameplayEffectLocalFX|HandleMeshGameplayEffect _localgameplayeffectvolume (Variables|GameplayEffects|GetBubbles%20Gameplay%20Effect) _returnvalue))
    (:Papaya.PlayerSparkles
      (Class|BGAPapayaGameplayEffectLocalFX|HandleMeshGameplayEffect _localgameplayeffectvolume (Variables|GameplayEffects|GetSparkles%20Gameplay%20Effect) _returnvalue))
    (:Papaya.PlayerHearts
      (Class|BGAPapayaGameplayEffectLocalFX|HandleMeshGameplayEffect _localgameplayeffectvolume (Variables|GameplayEffects|GetHearts%20Gameplay%20Effect) _returnvalue))
    (:Papaya.ConcertPlexus
      (Class|BGAPapayaGameplayEffectLocalFX|HandleMeshGameplayEffect _localgameplayeffectvolume (Variables|GameplayEffects|GetConcert%20Plexus%20Gameplay%20Effect) _returnvalue))
    (:Papaya.ConcertRain
      (Class|BGAPapayaGameplayEffectLocalFX|HandleMeshGameplayEffect _localgameplayeffectvolume (Variables|GameplayEffects|GetConcert%20Rain%20Local%20Gameplay%20Effect) _returnvalue)
      (Class|BGAPapayaGameplayEffectLocalFX|HandleMeshGameplayEffect _localgameplayeffectvolume (Variables|GameplayEffects|GetConcert%20Rain%20World%20Gameplay%20Effect) _returnvalue))))

(event Custom|ShowAudienceTease
  (params (Visible :type Boolean))
  ;[1 0.4115 0.4115] Moving Lights Sequence
  (Variables|StageSetup|SetActivateSeqMovingLights Visible))

(event Custom|NexusRaysAlpha
  (params (Alpha :type Float%20%28double-precision%29))
  ;[1 1 1] Nexus Rays Alpha
  (Variables|StageSetup|SetScreen%20Nexus%20Rays%20Opacity Alpha))

(event Custom|ShowConfetti
  (params (Visible :type Boolean))
  ;[1 0.3594 0.3594] Confetti
  (Variables|StageSetup|Confetti|SetConfettiActive Visible))

(event Custom|ControlTOD
  (bind _play_fade_time (Variables|TODManagement|GetPlay%20Fade%20Time))
  ;[0.2007 0 0.6] Take Control Time of Day Manager and set frozen TOD
  (TimeOfDay|SetTimeOfDayManagerVisibility (TimeOfDay|GetContextualTimeOfDayManager) true)
  (Utilities|BlendtoTargetTimeOfDay (Variables|TODManagement|GetDesired%20Time%20of%20Day) _play_fade_time
    (:Out
      (Utilities|Time|SetTimerbyEvent (EventDispatchers|CreateEvent :!event Custom|Freeze%20Time) _play_fade_time))))

(event Custom|AudienceTeaseFadeTime
  (params (MovingLightsInitialFadeTime :type Float%20%28double-precision%29))
  ;[1 0.4115 0.4115] Moving Lights Sequence
  (Variables|StageSetup|SetMovingLightsInitialFadeTime MovingLightsInitialFadeTime))

(event Custom|ConfettiRate
  (params (Rate :type Float%20%28double-precision%29))
  ;[1 0.3594 0.3594] Confetti
  (Variables|StageSetup|Confetti|SetConfettiSpawnRate Rate))

(event Custom|ShowScreenNexus
  (params (Visible :type Boolean))
  ;[1 1 1] Screen Nexus Toggle
  (Variables|StageSetup|SetScreen%20Nexus%20Activate Visible))

(event Custom|EnableBouncy
  (params (Enabled :type Boolean))
  ;[0.04979 1 0] FX Player Gameplay Cues
  ;;[1 1 1] Gameplay Element Added or Altered
  (Variables|StageSetup|SetActivatePlayerBouncy Enabled))

(event Custom|EnablePlayerBubbles
  (params (Enabled :type Boolean))
  ;[0.04979 1 0] FX Player Gameplay Cues
  ;;[1 1 1] FX Only
  (Variables|StageSetup|SetActivatePlayerBubbles Enabled))

(event Custom|ShowSRTetra
  (params (Visible :type Boolean))
  ;[0.2917 1.555 2] Tetra Screens
  ;;[1 1 1] Toggle SR Tetra Screen
  (bind _structout (Utilities|Struct|SetmembersinPapayaEffectsConcertTetraScreen (Variables|StageSetup|GetSR%20TetraScreen) Visible))
  (Variables|StageSetup|SetSR%20TetraScreen _structout))

(event Custom|SRTetraRotRate
  (params (Rate :type Float%20%28double-precision%29))
  ;[0.2917 1.555 2] Tetra Screens
  ;;[1 1 1] Set SR Cube Screen Rotation Rate
  (bind _structout (Utilities|Struct|SetmembersinPapayaEffectsConcertTetraScreen (Variables|StageSetup|GetSR%20TetraScreen) Rate))
  (Variables|StageSetup|SetSR%20TetraScreen _structout))

(event Custom|EnableSRTetraRot
  (params (Enabled :type Boolean))
  ;[0.2917 1.555 2] Tetra Screens
  ;;[1 1 1] Toggle SR Tetra Screen Rotation
  (bind _structout (Utilities|Struct|SetmembersinPapayaEffectsConcertTetraScreen (Variables|StageSetup|GetSR%20TetraScreen) Enabled))
  (Variables|StageSetup|SetSR%20TetraScreen _structout))

(event EventBeginPlay
  ;[1 0.2579 0] Event Begin Play and Initial Setup and resolve Soft References
  (Utilities|FlowControl|Sequence
    (:then_0
      (bind _returnvalue (Utilities|Time|SetTimerbyEvent (EventDispatchers|CreateEvent :!event Custom|Pulse) (Variables|ColorCycle|GetColorScrollRepeatTime) true))
      (Variables|ColorCycle|SetColor%20Cycling _returnvalue)
      (Networking|CanExecuteCosmeticEvents
        (:True
          (AudioAnalysis|BindEventtoOnSpectralAnalysisUpdated (WorldSubsystems|GetFortAudioAnalysisSubsystem) (EventDispatchers|CreateEvent :!event Custom|On%20Spectral%20Analysis%20Update)))))
    (:then_1
      (CallFunction|JIPSetPhysProps))))

(event Custom|FadeInMovingLights
  (params (Fade%20Time :type Float%20%28double-precision%29))
  ;[1 0.4115 0.4115] Moving Lights Sequence
  (Default|TimelineSetup Fade%20Time 0.0 (Variables|BP_Papaya_StageManager|GetFadeMovingLights) (Variables|StageSetup|GetMovingLightsAlpha))
  (bind (_direction _lerpcontroller) (Timeline|FadeMovingLights Play)
    (:Update
      (Variables|StageSetup|SetMovingLightsAlpha _lerpcontroller))))

(event Custom|Freeze%20Time
  ;[0.2007 0 0.6] Take Control Time of Day Manager and set frozen TOD
  (TimeOfDay|SetTimeOfDaySpeed 0.0 :!class /Script/FortniteGame.FortKismetLibrary))

(event Custom|EnableFireJump
  (params (Enabled :type Boolean))
  ;[0.04979 1 0] FX Player Gameplay Cues
  ;;[1 1 1] Gameplay Element Added or Altered
  (Variables|StageSetup|SetActivate%20Player%20Fire%20Jump Enabled))

(event Custom|EnablePlayerHearts
  (params (Enabled :type Boolean))
  ;[0.04979 1 0] FX Player Gameplay Cues
  ;;[1 1 1] FX Only
  (Variables|StageSetup|SetActivatePlayerHearts Enabled))

(event Custom|HazeAlpha
  (params (Alpha :type Float%20%28double-precision%29))
  ;[1 1 1] Screen Haze Alpha
  (Variables|StageSetup|SetScreen%20Haze%20Opacity Alpha))

(event Custom|ShowSideEQ
  (params (Visible :type Boolean))
  ;[1 1 1] Side EQ Screens
  (Variables|StageSetup|SetScreen%20Side%20Equalizers%20Activate Visible))

(event Custom|ShowDiscoLights
  (params (Visible :type Boolean))
  ;[1 0.3594 0.3594] Disco_Lights (Distant Pulsing Lights for Concert Area)
  (Variables|StageSetup|DiscoLights|SetDiscoLightsActive Visible))

(event Custom|TODSpeed
  (params (Speed :type Float%20%28double-precision%29))
  ;[0.2007 0 0.6] Take Control Time of Day Manager and set frozen TOD
  (TimeOfDay|SetTimeOfDaySpeed Speed :!class /Script/FortniteGame.FortKismetLibrary))

(event Custom|SetDesiredTOD
  (params (Desired%20Time%20of%20Day :type Float%20%28double-precision%29))
  ;[0.2007 0 0.6] Take Control Time of Day Manager and set frozen TOD
  (Variables|TODManagement|SetDesired%20Time%20of%20Day Desired%20Time%20of%20Day))

(event Custom|FadeOutMovingLights
  (params (Fade%20Time :type Float%20%28double-precision%29))
  ;[1 0.4115 0.4115] Moving Lights Sequence
  (Default|TimelineSetup Fade%20Time 1.0 (Variables|BP_Papaya_StageManager|GetFadeMovingLights) (Variables|StageSetup|GetMovingLightsAlpha))
  (Timeline|FadeMovingLights Reverse))

(event Custom|SetDurationTOD
  (params (Play%20Fade%20Time :type Float%20%28double-precision%29))
  ;[0.2007 0 0.6] Take Control Time of Day Manager and set frozen TOD
  (Variables|TODManagement|SetPlay%20Fade%20Time Play%20Fade%20Time))

(event Custom|EnablePlayerSparkles
  (params (Enabled :type Boolean))
  ;[0.04979 1 0] FX Player Gameplay Cues
  ;;[1 1 1] FX Only
  (Variables|StageSetup|SetActivatePlayerSparkles Enabled))

(event Custom|ShowSLTetra
  (params (Visible :type Boolean))
  ;[0.2917 1.555 2] Tetra Screens
  ;;[1 1 1] Toggle SL Tetra Screen
  (bind _structout (Utilities|Struct|SetmembersinPapayaEffectsConcertTetraScreen (Variables|StageSetup|GetSLTetraScreen) Visible))
  (Variables|StageSetup|SetSLTetraScreen _structout))

(event Custom|EnableSLTetraRot
  (params (Enabled :type Boolean))
  ;[0.2917 1.555 2] Tetra Screens
  ;;[1 1 1] Toggle SL Tetra Screen Rotation
  (bind _structout (Utilities|Struct|SetmembersinPapayaEffectsConcertTetraScreen (Variables|StageSetup|GetSLTetraScreen) Enabled))
  (Variables|StageSetup|SetSLTetraScreen _structout))

(event Custom|SLTetraRotRate
  (params (Rate :type Float%20%28double-precision%29))
  ;[0.2917 1.555 2] Tetra Screens
  ;;[1 1 1] Set SL Cube Screen Rotation Rate
  (bind _structout (Utilities|Struct|SetmembersinPapayaEffectsConcertTetraScreen (Variables|StageSetup|GetSLTetraScreen) Rate))
  (Variables|StageSetup|SetSLTetraScreen _structout))

(event Custom|SetMovingLightColorOffset
  (params (Color%20Offset :type Float%20%28double-precision%29))
  ;[1 0.4115 0.4115] Moving Lights Sequence
  (Variables|StageSetup|SetMovingLightColorOffset Color%20Offset))

(event Custom|ShowConcertRain
  (params (Visible :type Boolean))
  ;[0.04979 1 0] FX Player Gameplay Cues
  ;;[1 1 1] FX Only
  (Variables|StageSetup|SetActivate%20Concert%20Rain Visible))

(event Custom|ToggleVODTransition
  ;[0.2572 0 1] Effect to Cover VOD Transtions
  (Variables|StageSetup|SetActivateVODTransition (not (Variables|StageSetup|GetActivateVODTransition))))

(event Custom|SetMovingLightsAlpha
  (params (MovingLightsAlpha :type Float%20%28double-precision%29))
  ;[1 0.4115 0.4115] Moving Lights Sequence
  (Variables|StageSetup|SetMovingLightsAlpha MovingLightsAlpha))

(event Custom|FireworksRate
  (params (Rate :type Float%20%28double-precision%29))
  ;[1 0.3594 0.3594] Fireworks
  (Variables|StageSetup|Fireworks|SetFireworksSpawnRate Rate))

(event Custom|ShowConcertLights
  (params (Visible :type Boolean))
  ;[1 1 1] Concert Lights
  (Variables|StageSetup|SetConcert%20Lighting%20Activate Visible))

(event Custom|ShowFireworks
  (params (Visible :type Boolean))
  ;[1 0.3594 0.3594] Fireworks
  (Variables|StageSetup|Fireworks|SetFireworksActive Visible))

(event Custom|SetTetraScreensRotRate
  (params (Rate :type Float%20%28double-precision%29))
  ;[0.2917 1.555 2] Tetra Screens
  ;;[1 1 1] Set Both Cube Screens Rotation Rates
  (CallFunction|SRTetraRotRate Rate)
  (CallFunction|SLTetraRotRate Rate))

(event Custom|EnableTetraScreenRot
  (params (Enabled :type Boolean))
  ;[0.2917 1.555 2] Tetra Screens
  ;;[1 1 1] Toggle Both Tetra Screens Rotation
  (CallFunction|EnableSLTetraRot Enabled)
  (CallFunction|EnableSRTetraRot Enabled))

(event Custom|ShowTetraScreens
  (params (Visible :type Boolean))
  ;[0.2917 1.555 2] Tetra Screens
  ;;[1 1 1] Toggle Both Tetra Screens
  (CallFunction|ShowSLTetra Visible)
  (CallFunction|ShowSRTetra Visible))

(event Custom|ShowConcertPlexus
  (params (Visable :type Boolean))
  ;[0.04979 1 0] FX Player Gameplay Cues
  ;;[1 1 1] FX Only
  (Variables|StageSetup|SetActivate%20Plexus Visable))

(event Custom|HoloPlanesAlpha
  (params (Alpha :type Float%20%28double-precision%29))
  ;[0.3307 0.6 0] Hologram Planes
  (Variables|StageSetup|SetHologramPlanesAlpha Alpha))

(event Custom|On%20Spectral%20Analysis%20Update
  (params (Submix :type Sound%20Submix%20Object%20Reference) (Amplitudes :type Array%20of%20Float%20%28single-precision%29s))
  ;[0 0 1] Audio Analysis
  (if (Utilities|Equal(Object) Submix (Variables|Audio|GetAnalysisSubmix))
    (CallFunction|SpectralAnalysisUpdate)))

(event Custom|SparksRainRate
  (params (Rate :type Float%20%28double-precision%29))
  ;[1 0.3594 0.3594] Sparks Rain
  (Variables|StageSetup|SparksRain|SetSparksRainSpawnRate Rate))

(event Custom|ConcertActive
  ;[1 0.9624 0] Whole Stage Looks
  (Default|UpdateEverything (Variables|StageLooks|GetConcert_Look)))

(event Custom|Dark
  ;[1 0.9624 0] Whole Stage Looks
  (Default|UpdateEverything (Variables|StageLooks|GetDark_Look)))

(event Custom|ShowSparksRain
  (params (Visible :type Boolean))
  ;[1 0.3594 0.3594] Sparks Rain
  (Variables|StageSetup|SparksRain|SetSparksRainActive Visible))

(event Custom|ShowWorldGlobe
  (params (Visible :type Boolean))
  ;[1 1 1] World Globe
  (Variables|StageSetup|SetWorldGlobeVisible Visible))

(event Custom|FadeInHoloPlanes
  (params (FadeTime :type Float%20%28double-precision%29))
  ;[0.3307 0.6 0] Hologram Planes
  (Default|TimelineSetup FadeTime 0.0 (Variables|BP_Papaya_StageManager|GetFadeHoloPlane) (Variables|StageSetup|GetHologramPlanesAlpha))
  (bind (_direction _fade_0_1) (Timeline|FadeHoloPlane Play)
    (:Update
      (Variables|StageSetup|SetHologramPlanesAlpha _fade_0_1))))

(event Custom|Intermission
  ;[1 0.9624 0] Whole Stage Looks
  (Default|UpdateEverything (Variables|StageLooks|GetIntermission_Look)))

(event Custom|TOOMUCH
  ;[1 0.9624 0] Whole Stage Looks
  (Default|UpdateEverything (Variables|StageLooks|GetTooMuch_Look)))

(event Custom|FadeOutHoloPlanes
  (params (Fade%20Time :type Float%20%28double-precision%29))
  ;[0.3307 0.6 0] Hologram Planes
  (Default|TimelineSetup Fade%20Time 1.0 (Variables|BP_Papaya_StageManager|GetFadeHoloPlane) (Variables|StageSetup|GetHologramPlanesAlpha))
  (Timeline|FadeHoloPlane Reverse))

(event Custom|Pulse
  ;[1 0.04338 0] Color
  (if (not (Variables|ColorCycle|GetPulseStop))
    (Variables|StageSetup|Color|SetBase%20ColorParameter (Math|Interpolation|FInterpTo (Variables|StageSetup|Color|GetBase%20ColorParameter) (+ (Variables|StageSetup|Color|GetBase%20ColorParameter) (* (Variables|ColorCycle|GetColorScrollRepeatTime) (* (Variables|ColorCycle|GetTime%20Multiplier) 0.03125))) (Utilities|Time|GetWorldDeltaSeconds) 1e+01))))

(event Custom|ShowDiscoBalls
  (params (Visible :type Boolean))
  ;[0.2917 1.555 2] Disco Balls
  ;;[1 1 1] Toggle Disco Balls
  (bind _structout (Utilities|Struct|SetmembersinPapayaEffectsConcertDiscoBalls (Variables|StageSetup|GetDisco%20Ball) Visible))
  (Variables|StageSetup|SetDisco%20Ball _structout))

(event Custom|ScreenShardRate
  (params (Rate :type Float%20%28double-precision%29))
  ;[1 1 1] Screen Shards
  (Variables|Default|SetScreenShardsRate Rate))

(event Custom|ShowScreenShards
  (params (Visible :type Boolean))
  ;[1 1 1] Screen Shards
  (Variables|Default|SetScreenShards Visible))

(event Custom|Preshow
  ;[1 0.9624 0] Whole Stage Looks
  (Default|UpdateEverything (Variables|StageLooks|GetPreshow_Look)))

(event Custom|StageIntensity
  (params (Intensity :type Float%20%28double-precision%29))
  ;[1 0.04338 0] Color
  (Variables|StageSetup|Color|SetBase%20Color%20Mix Intensity))

(event Custom|StageTint
  (params (Tint :type Float%20%28double-precision%29))
  ;[1 0.04338 0] Color
  (Variables|StageSetup|Color|SetBase%20ColorParameter Tint))

(event Custom|IntermissionIntensity
  (params (Intensity :type Float%20%28double-precision%29))
  ;[1 0.04338 0] Color
  (Variables|StageSetup|Color|SetIntermission%20Color%20Mix Intensity))

(event Custom|BeginColorScroll
  ;[1 0.04338 0] Color
  (if (Utilities|Time|IsTimerActivebyHandle (Variables|ColorCycle|GetColor%20Cycling))
    (Development|PrintString "Timer already ACTIVE")
    (else
      (bind _returnvalue (Utilities|Time|SetTimerbyEvent (EventDispatchers|CreateEvent :!event Custom|Pulse) (Variables|ColorCycle|GetColorScrollRepeatTime) true))
      (Variables|ColorCycle|SetColor%20Cycling _returnvalue)
      (Variables|ColorCycle|SetPulseStop false))))

(event Custom|DiscoBallNum
  (params (Amount :type Float%20%28double-precision%29))
  ;[0.2917 1.555 2] Disco Balls
  ;;[1 1 1] Set Number of Disco Balls
  (bind _structout (Utilities|Struct|SetmembersinPapayaEffectsConcertDiscoBalls (Variables|StageSetup|GetDisco%20Ball) Amount))
  (Variables|StageSetup|SetDisco%20Ball _structout))

(event Custom|DiscoCluster
  (params (Amount :type Float%20%28double-precision%29))
  ;[0.2917 1.555 2] Disco Balls
  ;;[1 1 1] Set Number of Disco Ball Clusters
  (bind _structout (Utilities|Struct|SetmembersinPapayaEffectsConcertDiscoBalls (Variables|StageSetup|GetDisco%20Ball) Amount))
  (Variables|StageSetup|SetDisco%20Ball _structout))

(event Custom|ShowLaser1
  (params (Visible :type Boolean))
  ;[0.2917 1.555 2] Laser Effects
  (Variables|StageSetup|Lasers|SetLaser1 Visible))

(event Custom|ShowLasers
  (params (Visible :type Boolean))
  ;[0.2917 1.555 2] Laser Effects
  (CallFunction|ShowLaser1 Visible)
  (CallFunction|ShowLaser2 Visible)
  (CallFunction|ShowLaser3 Visible)
  (CallFunction|ShowLaser4 Visible)
  (CallFunction|ShowLaser5 Visible)
  (CallFunction|ShowLaser6 Visible))

(event Custom|ControlScreenFeedTVs
  (params (Mode :type Integer))
  ;[1 1 1] TVs
  (Variables|StageSetup|SetTVScreenMode Mode))

(event Custom|IntermissionTint
  (params (Tint :type Float%20%28double-precision%29))
  ;[1 0.04338 0] Color
  (Variables|StageSetup|Color|SetIntermission%20Color%20Parameter Tint))

(event Custom|RestoreLaser1
  ;[0.2917 1.555 2] Laser Effects
  (Variables|StageSetup|Lasers|SetLaser1_CustomColorUse 0.0))

(event Custom|ShowLaser2
  (params (Visible :type Boolean))
  ;[0.2917 1.555 2] Laser Effects
  (Variables|StageSetup|Lasers|SetLaser2 Visible))

(event Custom|RestoreLaser2
  ;[0.2917 1.555 2] Laser Effects
  (Variables|StageSetup|Lasers|SetLaser2_CustomColorUse 0.0))

(event Custom|StageDesat
  (params (Desat%20Value :type Float%20%28double-precision%29))
  ;[1 0.04338 0] Color
  (Variables|StageSetup|Color|SetDesaturation Desat%20Value))

(event Custom|Dancer1Emote
  (params (Byte :type Byte))
  ;[0.2917 1.555 2] Hologram Dancers
  ;;[1 1 1] Dancer 1 - SR Platform
  (Variables|StageSetup|Dancer1|SetDancer1_Emote (Utilities|Enum|BytetoEnumENUM_Papaya_Dancer_Emote Byte)))

(event Custom|ShowDancer1
  (params (Visible :type Boolean))
  ;[0.2917 1.555 2] Hologram Dancers
  ;;[1 1 1] Dancer 1 - SR Platform
  (Variables|StageSetup|Dancer1|SetDancer1Active Visible))

(event Custom|ShowLaser3
  (params (Visible :type Boolean))
  ;[0.2917 1.555 2] Laser Effects
  (Variables|StageSetup|Lasers|SetLaser3 Visible))

(event Custom|ShowDancer3
  (params (Visible :type Boolean))
  ;[0.2917 1.555 2] Hologram Dancers
  ;;[1 1 1] Dancer 3 - SR Catwalk
  (Variables|StageSetup|Dancer3|SetDancer3Active Visible))

(event Custom|Dancer3Emote
  (params (Byte :type Byte))
  ;[0.2917 1.555 2] Hologram Dancers
  ;;[1 1 1] Dancer 3 - SR Catwalk
  (Variables|StageSetup|Dancer3|SetDancer3_Emote (Utilities|Enum|BytetoEnumENUM_Papaya_Dancer_Emote Byte)))

(event Custom|RestoreLaser3
  ;[0.2917 1.555 2] Laser Effects
  (Variables|StageSetup|Lasers|SetLaser3_CustomColorUse 0.0))

(event Custom|SetFluorescentLightMode
  (params (Mode :type Integer))
  ;[1 1 1] Fluorescent Lights
  (Variables|StageSetup|SetFluorescentLightMode Mode))

(event Custom|DiscoMax
  (params (Amount :type Float%20%28double-precision%29))
  ;[0.2917 1.555 2] Disco Balls
  ;;[1 1 1] Set DiscoBall Max Scale
  (bind _structout (Utilities|Struct|SetmembersinPapayaEffectsConcertDiscoBalls (Variables|StageSetup|GetDisco%20Ball) Amount))
  (Variables|StageSetup|SetDisco%20Ball _structout))

(event Custom|DiscoMin
  (params (Amount :type Float%20%28double-precision%29))
  ;[0.2917 1.555 2] Disco Balls
  ;;[1 1 1] Set Disco Ball Min Scale
  (bind _structout (Utilities|Struct|SetmembersinPapayaEffectsConcertDiscoBalls (Variables|StageSetup|GetDisco%20Ball) Amount))
  (Variables|StageSetup|SetDisco%20Ball _structout))

(event Custom|CancelColorScroll
  (params (Base%20ColorParameter :type Float%20%28double-precision%29))
  ;[1 0.04338 0] Color
  (Variables|StageSetup|Color|SetBase%20ColorParameter (Math|Float|SelectFloat (Variables|StageSetup|Color|GetBase%20ColorParameter) Base%20ColorParameter (< Base%20ColorParameter 0.0)))
  (Utilities|Time|ClearandInvalidateTimerbyHandle (Variables|ColorCycle|GetColor%20Cycling))
  (Variables|ColorCycle|SetPulseStop true))

(event Custom|ColorScroll
  (params (Time%20Multiplier :type Float%20%28double-precision%29))
  ;[1 0.04338 0] Color
  (Variables|ColorCycle|SetTime%20Multiplier Time%20Multiplier))

(event Custom|ShowLaser4
  (params (Visible :type Boolean))
  ;[0.2917 1.555 2] Laser Effects
  (Variables|StageSetup|Lasers|SetLaser4 Visible))

(event Custom|RestoreLaser4
  ;[0.2917 1.555 2] Laser Effects
  (Variables|StageSetup|Lasers|SetLaser4_CustomColorUse 0.0))

(event Custom|Dancer1Mesh
  (params (Byte :type Byte))
  ;[0.2917 1.555 2] Hologram Dancers
  ;;[1 1 1] Dancer 1 - SR Platform
  (Variables|StageSetup|Dancer1|SetDancer1_Mesh (Utilities|Enum|BytetoEnumENUM_Papaya_Dancer Byte)))

(event Custom|Dancer1Scale
  (params (Scale :type Float%20%28double-precision%29))
  ;[0.2917 1.555 2] Hologram Dancers
  ;;[1 1 1] Dancer 1 - SR Platform
  (Variables|StageSetup|Dancer1|SetDancer1_Scale Scale))

(event Custom|Dancer3Mesh
  (params (Byte :type Byte))
  ;[0.2917 1.555 2] Hologram Dancers
  ;;[1 1 1] Dancer 3 - SR Catwalk
  (Variables|StageSetup|Dancer3|SetDancer3_Mesh (Utilities|Enum|BytetoEnumENUM_Papaya_Dancer Byte)))

(event Custom|Dancer3Scale
  (params (Scale :type Float%20%28double-precision%29))
  ;[0.2917 1.555 2] Hologram Dancers
  ;;[1 1 1] Dancer 3 - SR Catwalk
  (Variables|StageSetup|Dancer3|SetDancer3_Scale Scale))

(event Custom|RestoreLaser5
  ;[0.2917 1.555 2] Laser Effects
  (Variables|StageSetup|Lasers|SetLaser5_CustomColorUse 0.0))

(event Custom|ShowLaser5
  (params (Visible :type Boolean))
  ;[0.2917 1.555 2] Laser Effects
  (Variables|StageSetup|Lasers|SetLaser5 Visible))

(event Custom|RestoreLaser6
  ;[0.2917 1.555 2] Laser Effects
  (Variables|StageSetup|Lasers|SetLaser6_CustomColorUse 0.0))

(event Custom|ShowLaser6
  (params (Visible :type Boolean))
  ;[0.2917 1.555 2] Laser Effects
  (Variables|StageSetup|Lasers|SetLaser6 Visible))

(event Custom|SingularDiscoBall
  ;[0.2917 1.555 2] Disco Balls
  ;;[1 1 1] One Disco Ball to rule them all
  (bind _structout (Utilities|Struct|SetmembersinPapayaEffectsConcertDiscoBalls (Variables|StageSetup|GetDisco%20Ball) true 1.0 0.0 8.0 8.0 0.0))
  (Variables|StageSetup|SetDisco%20Ball _structout))

(event Custom|DiscoPivotScale
  (params (Amount :type Float%20%28double-precision%29))
  ;[0.2917 1.555 2] Disco Balls
  ;;[1 1 1] Disco Ball Rotation Pivot Offset Scale
  (bind _structout (Utilities|Struct|SetmembersinPapayaEffectsConcertDiscoBalls (Variables|StageSetup|GetDisco%20Ball) Amount))
  (Variables|StageSetup|SetDisco%20Ball _structout))

(event Custom|ShowDancer2
  (params (Visible :type Boolean))
  ;[0.2917 1.555 2] Hologram Dancers
  ;;[1 1 1] Dancer 2 - SL Platform
  (Variables|StageSetup|Dancer2|SetDancer2Active Visible))

(event Custom|ShowDancer4
  (params (Visible :type Boolean))
  ;[0.2917 1.555 2] Hologram Dancers
  ;;[1 1 1] Dancer 4 - SL Catwalk
  (Variables|StageSetup|Dancer4|SetDancer4Active Visible))

(event Custom|Dancer4Emote
  (params (Byte :type Byte))
  ;[0.2917 1.555 2] Hologram Dancers
  ;;[1 1 1] Dancer 4 - SL Catwalk
  (Variables|StageSetup|Dancer4|SetDancer4_Emote (Utilities|Enum|BytetoEnumENUM_Papaya_Dancer_Emote Byte)))

(event Custom|Dancer2Emote
  (params (Byte :type Byte))
  ;[0.2917 1.555 2] Hologram Dancers
  ;;[1 1 1] Dancer 2 - SL Platform
  (Variables|StageSetup|Dancer2|SetDancer2_Emote (Utilities|Enum|BytetoEnumENUM_Papaya_Dancer_Emote Byte)))

(event Custom|LaserSensor
  (params (Sensitivity%20Multiplier :type Float%20%28double-precision%29))
  ;[0.2917 1.555 2] Laser Effects
  (Variables|StageSetup|Lasers|SetLaserSensitivity Sensitivity%20Multiplier))

(event Custom|TetraScreenTint
  (params (Tint%20Amount :type Float%20%28double-precision%29))
  ;[1 0.04338 0] Tetra Screen Video Display Commands
  (Variables|StageSetup|SetTetraTint Tint%20Amount))

(event Custom|ChangeKaleidoscopeValue
  (params (Value :type Float%20%28double-precision%29))
  ;[1 1 1] Kaleidoscope
  (Variables|StageSetup|SetKaleidoscopeValue (Math|Float|NormalizetoRange Value 0.0 1.0)))

(event Custom|ShowSpriteDome
  (params (bShow :type Boolean))
  ;[1 1 1] SpriteDome
  (Variables|StageSetup|SetSpriteDomeActive bShow))

(event Custom|Dancer2Mesh
  (params (Byte :type Byte))
  ;[0.2917 1.555 2] Hologram Dancers
  ;;[1 1 1] Dancer 2 - SL Platform
  (Variables|StageSetup|Dancer2|SetDancer2_Mesh (Utilities|Enum|BytetoEnumENUM_Papaya_Dancer Byte)))

(event Custom|Dancer2Scale
  (params (Scale :type Float%20%28double-precision%29))
  ;[0.2917 1.555 2] Hologram Dancers
  ;;[1 1 1] Dancer 2 - SL Platform
  (Variables|StageSetup|Dancer2|SetDancer2_Scale Scale))

(event Custom|Dancer4Mesh
  (params (Byte :type Byte))
  ;[0.2917 1.555 2] Hologram Dancers
  ;;[1 1 1] Dancer 4 - SL Catwalk
  (Variables|StageSetup|Dancer4|SetDancer4_Mesh (Utilities|Enum|BytetoEnumENUM_Papaya_Dancer Byte)))

(event Custom|Dancer4Scale
  (params (Scale :type Float%20%28double-precision%29))
  ;[0.2917 1.555 2] Hologram Dancers
  ;;[1 1 1] Dancer 4 - SL Catwalk
  (Variables|StageSetup|Dancer4|SetDancer4_Scale Scale))

(event Custom|TetraUVSetup
  (params (UVSetup :type Float%20%28double-precision%29))
  ;[1 0.04338 0] Tetra Screen Video Display Commands
  (Variables|StageSetup|SetTetraUV UVSetup))

(event Custom|RestoreLasers
  ;[0.2917 1.555 2] Laser Effects
  (CallFunction|RestoreLaser1)
  (CallFunction|RestoreLaser2)
  (CallFunction|RestoreLaser3)
  (CallFunction|RestoreLaser4)
  (CallFunction|RestoreLaser5)
  (CallFunction|RestoreLaser6))

(event Custom|DiscoBall
  (params (Active :type Boolean) (Amount%20of%20Disco%20Balls :type Float%20%28double-precision%29) (Number%20of%20Clusters :type Float%20%28double-precision%29) (MAX%20Disco%20Ball%20Size :type Float%20%28double-precision%29) (MIN%20Disco%20Ball%20Size :type Float%20%28double-precision%29) (Rotation%20Pivot%20Offset%20Scale :type Float%20%28double-precision%29))
  ;[0.2917 1.555 2] Disco Balls
  ;;[1 1 1] Allow Complete Change to Disco Balls in one command, multiple inputs
  (bind _structout (Utilities|Struct|SetmembersinPapayaEffectsConcertDiscoBalls (Variables|StageSetup|GetDisco%20Ball) Active Amount%20of%20Disco%20Balls Number%20of%20Clusters MAX%20Disco%20Ball%20Size MIN%20Disco%20Ball%20Size Rotation%20Pivot%20Offset%20Scale))
  (Variables|StageSetup|SetDisco%20Ball _structout))

(event Custom|RestoreDiscoBall
  ;[0.2917 1.555 2] Disco Balls
  ;;[1 1 1] Restore Base Look for orbitting Disco Balls
  (bind _structout (Utilities|Struct|SetmembersinPapayaEffectsConcertDiscoBalls (Variables|StageSetup|GetDisco%20Ball) true 15.0 5.0 3.0 1.0 1.0))
  (Variables|StageSetup|SetDisco%20Ball _structout))

(event Custom|ShowCartoonOutlinePostProcess
  (params (bShow :type Boolean))
  ;[1 1 1] Borderlands Post Process
  (Variables|StageSetup|SetCartoonOutlinePostProcessActive bShow))

(event Custom|SetCartoonOutlinePreset
  (params (Index :type Integer))
  ;[1 1 1] Borderlands Post Process
  (Variables|StageSetup|SetCartoonOutlinePreset Index))

(event Custom|ShowDancers
  (params (Visible :type Boolean))
  ;[0.2917 1.555 2] Hologram Dancers
  ;;[1 1 1] All Dancers
  (CallFunction|ShowDancer1 Visible)
  (CallFunction|ShowDancer2 Visible)
  (CallFunction|ShowDancer3 Visible)
  (CallFunction|ShowDancer4 Visible))

(event Custom|ColorLasers
  (params (R :type Float%20%28double-precision%29) (G :type Float%20%28double-precision%29) (B :type Float%20%28double-precision%29) (A :type Float%20%28double-precision%29))
  ;[0.2917 1.555 2] Laser Effects
  (CallFunction|ColorLaser1 R G B A)
  (CallFunction|ColorLaser2 R G B A)
  (CallFunction|ColorLaser3 R G B A)
  (CallFunction|ColorLaser4 R G B A)
  (CallFunction|ColorLaser5 R G B A)
  (CallFunction|ColorLaser6 R G B A))

(event Custom|ToggleMovieScreenPost
  (bind _activate_movie_post (Variables|StageSetup|MovieScreen|GetActivate%20Movie%20Post))
  ;[0.003357 0.148 0.599] Movie Screen Controls
  (Variables|StageSetup|MovieScreen|SetActivate%20Movie%20Post (not _activate_movie_post))
  (Variables|StageSetup|MovieScreen|SetMoviePostBlendWeight (select _activate_movie_post 1.0 0.0)))

(event Custom|ChangeMovieAspectRatio
  (params (Desired%20Aspect%20Ratio :type Byte))
  ;[0.003357 0.148 0.599] Movie Screen Controls
  (Variables|StageSetup|MovieScreen|SetAspectRatioScale (Utilities|Select :Index (Utilities|Enum|BytetoEnumENUM_Papaya_AspectRatios Desired%20Aspect%20Ratio) :NewEnumerator5 1.33 :NewEnumerator0 1.0 :NewEnumerator1 1.1 :NewEnumerator7 0.96 :NewEnumerator4 0.93 :NewEnumerator3 1.26 :NewEnumerator8 0.77 :NewEnumerator2 0.75 :NewEnumerator9 0.74 :NewEnumerator10 0.73 :NewEnumerator11 0.63)))

(event Custom|DancersMesh
  (params (Byte :type Byte))
  ;[0.2917 1.555 2] Hologram Dancers
  ;;[1 1 1] All Dancers
  (CallFunction|Dancer1Mesh Byte)
  (CallFunction|Dancer2Mesh Byte)
  (CallFunction|Dancer3Mesh Byte)
  (CallFunction|Dancer4Mesh Byte))

(event Custom|DancersEmote
  (params (Byte :type Byte))
  ;[0.2917 1.555 2] Hologram Dancers
  ;;[1 1 1] All Dancers
  (CallFunction|Dancer1Emote Byte)
  (CallFunction|Dancer2Emote Byte)
  (CallFunction|Dancer3Emote Byte)
  (CallFunction|Dancer4Emote Byte))

(event Custom|ColorLaser1
  (params (Red :type Float%20%28double-precision%29) (Green :type Float%20%28double-precision%29) (Blue :type Float%20%28double-precision%29) (A :type Float%20%28double-precision%29))
  ;[0.2917 1.555 2] Laser Effects
  (Variables|StageSetup|Lasers|SetLaser1_Color (Utilities|Struct|MakeLinearColor Red Green Blue A))
  (Variables|StageSetup|Lasers|SetLaser1_CustomColorUse 1.0))

(event Custom|ColorLaser4
  (params (Red :type Float%20%28double-precision%29) (Green :type Float%20%28double-precision%29) (Blue :type Float%20%28double-precision%29) (A :type Float%20%28double-precision%29))
  ;[0.2917 1.555 2] Laser Effects
  (Variables|StageSetup|Lasers|SetLaser4_Color (Utilities|Struct|MakeLinearColor Red Green Blue A))
  (Variables|StageSetup|Lasers|SetLaser4_CustomColorUse 1.0))

(event Custom|FadeInMoviePost
  (params (Fade :type Float%20%28double-precision%29))
  ;[0.003357 0.148 0.599] Movie Screen Controls
  (Variables|StageSetup|MovieScreen|SetActivate%20Movie%20Post true)
  (Default|TimelineSetup Fade 0.0 (Variables|BP_Papaya_StageManager|GetMoviePostFade) (Variables|StageSetup|MovieScreen|GetMoviePostBlendWeight))
  (bind (_direction _control_track) (Timeline|MoviePostFade Play)
    (:Update
      (Variables|StageSetup|MovieScreen|SetMoviePostBlendWeight _control_track))
    (:Finished
      (switch Utilities|FlowControl|Switch|SwitchonETimelineDirection _direction
        (:Backward
          (Variables|StageSetup|MovieScreen|SetActivate%20Movie%20Post true))))))

(event Custom|DancersScale
  (params (Scale :type Float%20%28double-precision%29))
  ;[0.2917 1.555 2] Hologram Dancers
  ;;[1 1 1] All Dancers
  (CallFunction|Dancer1Scale Scale)
  (CallFunction|Dancer2Scale Scale)
  (CallFunction|Dancer3Scale Scale)
  (CallFunction|Dancer4Scale Scale))

(event Custom|ColorLaser2
  (params (Red :type Float%20%28double-precision%29) (Green :type Float%20%28double-precision%29) (Blue :type Float%20%28double-precision%29) (A :type Float%20%28double-precision%29))
  ;[0.2917 1.555 2] Laser Effects
  (Variables|StageSetup|Lasers|SetLaser2_Color (Utilities|Struct|MakeLinearColor Red Green Blue A))
  (Variables|StageSetup|Lasers|SetLaser2_CustomColorUse 1.0))

(event Custom|ColorLaser5
  (params (Red :type Float%20%28double-precision%29) (Green :type Float%20%28double-precision%29) (Blue :type Float%20%28double-precision%29) (A :type Float%20%28double-precision%29))
  ;[0.2917 1.555 2] Laser Effects
  (Variables|StageSetup|Lasers|SetLaser5_Color (Utilities|Struct|MakeLinearColor Red Green Blue A))
  (Variables|StageSetup|Lasers|SetLaser5_CustomColorUse 1.0))

(event Custom|FadeOutMoviePost
  (params (Fade :type Float%20%28double-precision%29))
  ;[0.003357 0.148 0.599] Movie Screen Controls
  (Default|TimelineSetup Fade 1.0 (Variables|BP_Papaya_StageManager|GetMoviePostFade) (Variables|StageSetup|MovieScreen|GetMoviePostBlendWeight))
  (Timeline|MoviePostFade Reverse))

(event Custom|ColorLaser3
  (params (Red :type Float%20%28double-precision%29) (Green :type Float%20%28double-precision%29) (Blue :type Float%20%28double-precision%29) (A :type Float%20%28double-precision%29))
  ;[0.2917 1.555 2] Laser Effects
  (Variables|StageSetup|Lasers|SetLaser3_Color (Utilities|Struct|MakeLinearColor Red Green Blue A))
  (Variables|StageSetup|Lasers|SetLaser3_CustomColorUse 1.0))

(event Custom|ColorLaser6
  (params (Red :type Float%20%28double-precision%29) (Green :type Float%20%28double-precision%29) (Blue :type Float%20%28double-precision%29) (A :type Float%20%28double-precision%29))
  ;[0.2917 1.555 2] Laser Effects
  (Variables|StageSetup|Lasers|SetLaser6_Color (Utilities|Struct|MakeLinearColor Red Green Blue A))
  (Variables|StageSetup|Lasers|SetLaser6_CustomColorUse 1.0))

(event Custom|FadeMovieAspectRatio
  (params (Desired%20Aspect%20Ratio :type Byte) (FadeTime :type Float%20%28double-precision%29))
  ;[0.003357 0.148 0.599] Movie Screen Controls
  (Components|Timeline|SetPlayRate (Variables|BP_Papaya_StageManager|GetAspectRatioChange) (/ 1.0 (select (<= FadeTime 0.0) 0.1 FadeTime)))
  (Variables|StageSetup|MovieScreen|SetCurrentAspect (Variables|StageSetup|MovieScreen|GetAspectRatioScale))
  (bind (_direction _newtrack_0) (Timeline|AspectRatioChange PlayFromStart)
    (:Update
      (Variables|StageSetup|MovieScreen|SetAspectRatioScale (Math|Float|Lerp (Variables|StageSetup|MovieScreen|GetCurrentAspect) (Utilities|Select :Index (Utilities|Enum|BytetoEnumENUM_Papaya_AspectRatios Desired%20Aspect%20Ratio) :NewEnumerator5 1.33 :NewEnumerator0 1.0 :NewEnumerator1 1.1 :NewEnumerator7 0.96 :NewEnumerator4 0.93 :NewEnumerator3 1.26 :NewEnumerator8 0.77 :NewEnumerator2 0.75 :NewEnumerator9 0.74 :NewEnumerator10 0.73 :NewEnumerator11 0.63) _newtrack_0)))))

(event Custom|Glow1Alpha
  (params (Alpha :type Float%20%28double-precision%29))
  ;[0.2917 1.555 2] GLOW Strips
  ;;[1 1 1] Individual Strips
  (Variables|StageSetup|GlowStrips|SetScreen%20Curve%201%20Alpha Alpha))

(event Custom|GlowsAlpha
  (params (Alpha :type Float%20%28double-precision%29))
  ;[0.2917 1.555 2] GLOW Strips
  ;;[1 1 1] Set Alpha on ALL Glow Strips
  (CallFunction|Glow1Alpha Alpha)
  (CallFunction|Glow2Alpha Alpha)
  (CallFunction|Glow3Alpha Alpha)
  (CallFunction|Glow4Alpha Alpha)
  (CallFunction|GlowCenterAlpha Alpha)
  (CallFunction|GlowSRAlpha Alpha)
  (CallFunction|GlowSLAlpha Alpha))

(event Custom|Glow2Alpha
  (params (Alpha :type Float%20%28double-precision%29))
  ;[0.2917 1.555 2] GLOW Strips
  ;;[1 1 1] Individual Strips
  (Variables|StageSetup|GlowStrips|SetScreen%20Curve%202%20Alpha Alpha))

(event Custom|Glow3Alpha
  (params (Alpha :type Float%20%28double-precision%29))
  ;[0.2917 1.555 2] GLOW Strips
  ;;[1 1 1] Individual Strips
  (Variables|StageSetup|GlowStrips|SetScreen%20Curve%203%20Alpha Alpha))

(event Custom|Glow4Alpha
  (params (Alpha :type Float%20%28double-precision%29))
  ;[0.2917 1.555 2] GLOW Strips
  ;;[1 1 1] Individual Strips
  (Variables|StageSetup|GlowStrips|SetScreen%20Curve%204%20Alpha Alpha))

(event Custom|GlowSRAlpha
  (params (Alpha :type Float%20%28double-precision%29))
  ;[0.2917 1.555 2] GLOW Strips
  ;;[1 1 1] Individual Strips
  (Variables|StageSetup|GlowStrips|SetSR%20Entry%20GlowShape Alpha))

(event Custom|GlowCenterAlpha
  (params (Alpha :type Float%20%28double-precision%29))
  ;[0.2917 1.555 2] GLOW Strips
  ;;[1 1 1] Individual Strips
  (Variables|StageSetup|GlowStrips|SetCenter%20Entry%20GlowShape Alpha))

(event Custom|GlowSLAlpha
  (params (Alpha :type Float%20%28double-precision%29))
  ;[0.2917 1.555 2] GLOW Strips
  ;;[1 1 1] Individual Strips
  (Variables|StageSetup|GlowStrips|SetSL%20Entry%20GlowShape Alpha))

(event Custom|ToggleMainScreen
  (bind _main_screen_active (Variables|StageSetup|GetMain%20Screen%20Active))
  (bind _returnvalue (not _main_screen_active))
  (bind _returnvalue_2 (Math|Conversions|ToFloat(Boolean) _returnvalue))
  ;[1 0.3906 0.3906] Toggle Main Screen Visibility
  (Variables|Default|SetScreenShards _returnvalue)
  (Variables|StageSetup|SetScreen%20Nexus%20Activate _returnvalue)
  (Variables|StageSetup|GlowStrips|SetScreen%20Curve%201%20Alpha _returnvalue_2)
  (Variables|StageSetup|GlowStrips|SetScreen%20Curve%202%20Alpha _returnvalue_2)
  (Variables|StageSetup|GlowStrips|SetScreen%20Curve%203%20Alpha _returnvalue_2)
  (Variables|StageSetup|GlowStrips|SetScreen%20Curve%204%20Alpha _returnvalue_2)
  (Variables|StageSetup|SetScreen%20Nexus%20Rays%20Opacity _returnvalue_2)
  (Variables|StageSetup|SetScreen%20Haze%20Opacity _returnvalue_2)
  (Variables|StageSetup|SetMain%20Screen%20Active _returnvalue))

(event Custom|FadeOutMainscreen
  ;[1 0.4531 0.4531] Fade Main Screen via Scale
  (Variables|StageSetup|SetActivate%20MainScreen%20Rift true)
  (bind (_direction _control) (Timeline|MainScreenRemoval PlayFromStart)
    (:Update
      (Variables|StageSetup|SetMainScreenAlpha _control))
    (:Finished
      (switch Utilities|FlowControl|Switch|SwitchonETimelineDirection _direction
        (:Forward
          (Variables|Default|SetScreenShards false)
          (Variables|StageSetup|SetScreen%20Nexus%20Activate false)
          (Variables|StageSetup|SetMain%20Screen%20Active false)
          (Variables|StageSetup|SetActivate%20MainScreen%20Rift false)
          (Networking|SwitchHasAuthority
            (:Authority
              (CallFunction|MainScreenActiveChanged))))
        (:Backward
          (Variables|Default|SetScreenShards true))))))

(event Custom|FadeInMainScreen
  ;[1 0.4531 0.4531] Fade Main Screen via Scale
  (Variables|StageSetup|SetActivate%20MainScreen%20Rift true)
  (Variables|StageSetup|SetScreen%20Nexus%20Activate true)
  (Variables|StageSetup|SetMain%20Screen%20Active true)
  (Timeline|MainScreenRemoval ReverseFromEnd))

(event Custom|JIPSetPhysProps
  (bind _timescreenphysprops (Variables|Default|GetTimeScreenPhysProps))
  ;[1 1 1] Screen Position/Rotation/Scale Control
  (if (not (Math|Boolean|ORBoolean (Math|Vector|Equal(Vector) (Utilities|Array|Get(aref) (Variables|Default|GetMainScreenPhysProps)) "0, 0, 0" 0.0001) (Math|Vector|Equal(Vector) (Utilities|Array|Get(aref) (Variables|Default|GetMainScreenPhysProps) 1) "0, 0, 0" 0.0001) (Math|Vector|Equal(Vector) (Utilities|Array|Get(aref) (Variables|Default|GetMainScreenPhysProps) 2) "0, 0, 0" 0.0001)))
    (Networking|CanExecuteCosmeticEvents
      (:True
        (Utilities|WaitforScreenPropsReady 2.0
          (:Valid
            (if (> _timescreenphysprops -1.0)
              (Components|Timeline|SetPlayRate (Variables|BP_Papaya_StageManager|GetMainScreenPhys) (/ 0.5 _timescreenphysprops))
              (else
                (Components|Timeline|SetPlayRate (Variables|BP_Papaya_StageManager|GetMainScreenPhys) 1.0)))
            (if (Variables|Default|GetAnimDirScreenPhysProps)
              (bind (_direction _newtrack_0) (Timeline|MainScreenPhys PlayFromStart)
                (:Update
                  (CallFunction|UpdateMainScreenPhysProps _newtrack_0)))
              (else
                (Timeline|MainScreenPhys ReverseFromEnd)))))))))

(event Custom|SetScreenPhysProps
  (params (Px :type Float%20%28double-precision%29) (Py :type Float%20%28double-precision%29) (Pz :type Float%20%28double-precision%29) (Rx :type Float%20%28double-precision%29) (Ry :type Float%20%28double-precision%29) (Rz :type Float%20%28double-precision%29) (Sx :type Float%20%28double-precision%29) (Sy :type Float%20%28double-precision%29) (Sz :type Float%20%28double-precision%29) (Time :type Float%20%28double-precision%29) (AnimDir :type Boolean) (Reset :type Boolean))
  (bind _timescreenphysprops (Variables|Default|GetTimeScreenPhysProps))
  ;[1 1 1] Screen Position/Rotation/Scale Control
  (Variables|Default|SetAnimDirScreenPhysProps AnimDir)
  (Variables|Default|SetResetScreenPhysProps Reset)
  (Variables|Default|SetTimeScreenPhysProps Time)
  (Utilities|Array|SetArrayElem (Variables|Default|GetMainScreenPhysProps) 0 (Math|Vector|MakeVector Px Py Pz))
  (Utilities|Array|SetArrayElem (Variables|Default|GetMainScreenPhysProps) 1 (Math|Vector|MakeVector Rx Ry Rz))
  (Utilities|Array|SetArrayElem (Variables|Default|GetMainScreenPhysProps) 2 (Math|Vector|MakeVector Sx Sy Sz))
  (Networking|CanExecuteCosmeticEvents
    (:True
      (Utilities|WaitforScreenPropsReady 2.0
        (:Valid
          (if (> _timescreenphysprops -1.0)
            (Components|Timeline|SetPlayRate (Variables|BP_Papaya_StageManager|GetMainScreenPhys) (/ 0.5 _timescreenphysprops))
            (else
              (Components|Timeline|SetPlayRate (Variables|BP_Papaya_StageManager|GetMainScreenPhys) 1.0)))
          (if (Variables|Default|GetAnimDirScreenPhysProps)
            (Timeline|MainScreenPhys PlayFromStart)
            (else
              (Timeline|MainScreenPhys ReverseFromEnd))))))))

(event Custom|ShowRedirectToMusicScreen
  (params (Show :type Boolean))
  ;[1 1 1] Redirects and Posters
  (Variables|IslandScreensDONOTEDIT!!|SetRedirectToMusicScreenActive Show))

(event Custom|ShowOchoEventPoster
  (params (Show :type Boolean))
  ;[1 1 1] Redirects and Posters
  (Variables|IslandScreensDONOTEDIT!!|SetOchoScreenActive Show))

(event Custom|ShowSpecialEventPoster2
  (params (Show :type Boolean))
  ;[1 1 1] Redirects and Posters
  (Variables|IslandScreensDONOTEDIT!!|SetSpecialEventPoster2Active Show))

(fn CreateScreenPhysData
  (params (Component :type Scene%20Component%20Object%20Reference) (ItemIndex :type Integer))
  (return (Math|Vector|Lerp(Vector) (Transformation|GetWorldLocation Component) (Math|Vector|SelectVector (Utilities|Array|Get(aref) (Variables|Default|GetOriginalScreenPhysPositions) ItemIndex) (CallFunction|CreateScreenVector (Utilities|Array|Get(aref) (Variables|Default|GetMainScreenPhysProps)) (Transformation|GetWorldLocation Component)) (Variables|Default|GetResetScreenPhysProps)) (Variables|Default|GetMainScreenPhysAlpha)) (Math|Rotator|Lerp(Rotator) (Transformation|GetWorldRotation Component) (Math|Rotator|MakeRotator (.x (Math|Vector|SelectVector (Utilities|Array|Get(aref) (Variables|Default|GetOriginalScreenPhysRotations) ItemIndex) (CallFunction|CreateScreenVector (Utilities|Array|Get(aref) (Variables|Default|GetMainScreenPhysProps) 1) (Math|Vector|MakeVector (.roll (Transformation|GetWorldRotation Component)) (.pitch (Transformation|GetWorldRotation Component)) (.yaw (Transformation|GetWorldRotation Component)))) (Variables|Default|GetResetScreenPhysProps))) (.y (Math|Vector|SelectVector (Utilities|Array|Get(aref) (Variables|Default|GetOriginalScreenPhysRotations) ItemIndex) (CallFunction|CreateScreenVector (Utilities|Array|Get(aref) (Variables|Default|GetMainScreenPhysProps) 1) (Math|Vector|MakeVector (.roll (Transformation|GetWorldRotation Component)) (.pitch (Transformation|GetWorldRotation Component)) (.yaw (Transformation|GetWorldRotation Component)))) (Variables|Default|GetResetScreenPhysProps))) (.z (Math|Vector|SelectVector (Utilities|Array|Get(aref) (Variables|Default|GetOriginalScreenPhysRotations) ItemIndex) (CallFunction|CreateScreenVector (Utilities|Array|Get(aref) (Variables|Default|GetMainScreenPhysProps) 1) (Math|Vector|MakeVector (.roll (Transformation|GetWorldRotation Component)) (.pitch (Transformation|GetWorldRotation Component)) (.yaw (Transformation|GetWorldRotation Component)))) (Variables|Default|GetResetScreenPhysProps)))) (Variables|Default|GetMainScreenPhysAlpha)) (Math|Vector|Lerp(Vector) (Transformation|GetWorldScale Component) (Math|Vector|SelectVector (Utilities|Array|Get(aref) (Variables|Default|GetOriginalScreenPhysScales) ItemIndex) (CallFunction|CreateScreenVector (Utilities|Array|Get(aref) (Variables|Default|GetMainScreenPhysProps) 2) (Transformation|GetWorldScale Component)) (Variables|Default|GetResetScreenPhysProps)) (Variables|Default|GetMainScreenPhysAlpha))))

(fn CreateScreenVector
  (params (InVector :type Vector) (CurrentVector :type Vector))
  (return (Math|Vector|MakeVector (Math|Float|SelectFloat (.x CurrentVector) (.x InVector) (== (.x InVector) (LocalVariables|GetNoChangeValue))) (Math|Float|SelectFloat (.y CurrentVector) (.y InVector) (== (.y InVector) (LocalVariables|GetNoChangeValue))) (Math|Float|SelectFloat (.z CurrentVector) (.z InVector) (== (.z InVector) (LocalVariables|GetNoChangeValue))))))

(fn EnableFireworksAudio
  (params (Enabled :type Boolean))
  (bind _audio_fireworks (Variables|Default|GetAudio_Fireworks))
  (bind _audio_fireworks_2 (Variables|Default|GetAudio_Fireworks))
  (bind _fireworksaudiomix (Variables|Audio|GetFireworksAudioMix))
  (if Enabled
    (if (Utilities|Enum|NotEqual(Enum) (Audio|Components|Audio|GetPlayState _audio_fireworks_2) "Playing")
      (Audio|Components|Audio|Play _audio_fireworks_2 0.0)
      (Audio|PushSoundMixModifier _fireworksaudiomix))
    (else
      (if (Audio|Components|Audio|IsPlaying _audio_fireworks)
        (Audio|Components|Audio|FadeOut _audio_fireworks 5.0 0.0 "Linear" :!class /Script/Engine.AudioComponent)
        (Audio|PopSoundMixModifier _fireworksaudiomix)))))

(fn MainScreenActiveChanged
  (bind _screen_actor (Variables|StageManagementDONOTEDIT!!|GetScreen%20Actor))
  (bind _output (Utilities|ResolveSoftReference _screen_actor))
  (bind _floating_large_shards (Variables|StageManagementDONOTEDIT!!|GetFloating%20Large%20Shards))
  (bind _output_2 (Utilities|ResolveSoftReference _floating_large_shards))
  (bind _side_screens_actor (Variables|StageManagementDONOTEDIT!!|GetSide%20Screens%20Actor))
  (bind _output_3 (Utilities|ResolveSoftReference _side_screens_actor))
  (bind _main_screen_active (Variables|StageSetup|GetMain%20Screen%20Active))
  (Utilities|FlowControl|Sequence
    (:then_0
      (Utilities|IsValid _output
        (:Is%20Valid
          (CallFunction|ToggleActorHiddenInGame _main_screen_active _output)
          (Collision|SetActorEnableCollision _output _main_screen_active))))
    (:then_1
      (Utilities|IsValid _output_2
        (:Is%20Valid
          (CallFunction|ToggleActorHiddenInGame _main_screen_active _output_2)
          (Collision|SetActorEnableCollision _output_2 _main_screen_active))))
    (:then_2
      (Utilities|IsValid _output_3
        (:Is%20Valid
          (CallFunction|ToggleActorHiddenInGame _main_screen_active _output_3)
          (Collision|SetActorEnableCollision _output_3 _main_screen_active))))))

(fn OnRep_ActivateConcertRain
  (bind _concert_rain_local_gameplay_effect (Variables|GameplayEffects|GetConcert%20Rain%20Local%20Gameplay%20Effect))
  (bind _concert_rain_world_gameplay_effect (Variables|GameplayEffects|GetConcert%20Rain%20World%20Gameplay%20Effect))
  (bind _outfortplayerpawns (Fortnite|GetFortPlayerPawns))
  (for _array_element _outfortplayerpawns
    (if (Variables|StageSetup|GetActivate%20Concert%20Rain)
      (GameplayEffects|ApplyGameplayEffectToSelf (Ability|GetAbilitySystemComponent _array_element) _concert_rain_local_gameplay_effect 0.0 (GameplayEffects|MakeEffectContext (Ability|GetAbilitySystemComponent _array_element)))
      (GameplayEffects|ApplyGameplayEffectToSelf (Ability|GetAbilitySystemComponent _array_element) _concert_rain_world_gameplay_effect 0.0 (GameplayEffects|MakeEffectContext (Ability|GetAbilitySystemComponent _array_element)))
      (else
        (GameplayEffects|RemoveActiveGameplayEffectbySourceEffect (Ability|GetAbilitySystemComponent _array_element) _concert_rain_local_gameplay_effect 0 -1)
        (GameplayEffects|RemoveActiveGameplayEffectbySourceEffect (Ability|GetAbilitySystemComponent _array_element) _concert_rain_world_gameplay_effect 0 -1)))))

(fn OnRep_ActivateMainScreenRift
  (Networking|CanExecuteCosmeticEvents
    (:True
      (Components|Activation|Activate (Variables|/Script/CascadeRuntime.Emitter|GetParticleSystemComponent (Variables|StageManagementDONOTEDIT!!|GetMainScreenRift)) false))))

(fn OnRep_ActivateMoviePost
  (bind _movie_screen_post (Variables|StageManagementDONOTEDIT!!|GetMovie%20Screen%20Post))
  (bind _output (Utilities|ResolveSoftReference _movie_screen_post))
  (Networking|CanExecuteCosmeticEvents
    (:True
      (Utilities|IsValid _output
        (:Is%20Valid
          (Variables|/Script/Engine.PostProcessVolume|SetbEnabled (Variables|StageSetup|MovieScreen|GetActivate%20Movie%20Post) _output))))))

(fn OnRep_ActivatePlayerFireJump
  (bind _lowg_gameplay_effect (Variables|GameplayEffects|GetLowG%20Gameplay%20Effect))
  (bind _firejump_gameplay_effect (Variables|GameplayEffects|GetFireJump%20Gameplay%20Effect))
  (bind _outfortplayerpawns (Fortnite|GetFortPlayerPawns))
  (for _array_element _outfortplayerpawns
    (if (Variables|StageSetup|GetActivate%20Player%20Fire%20Jump)
      (GameplayEffects|ApplyGameplayEffectToSelf :self (Ability|GetAbilitySystemComponent _array_element) :GameplayEffectClass _firejump_gameplay_effect :Level 0.0)
      (GameplayEffects|ApplyGameplayEffectToSelf :self (Ability|GetAbilitySystemComponent _array_element) :GameplayEffectClass _lowg_gameplay_effect :Level 0.0)
      (else
        (GameplayEffects|RemoveActiveGameplayEffectbySourceEffect (Ability|GetAbilitySystemComponent _array_element) _firejump_gameplay_effect 0 -1)
        (GameplayEffects|RemoveActiveGameplayEffectbySourceEffect (Ability|GetAbilitySystemComponent _array_element) _lowg_gameplay_effect 0 -1)))))

(fn OnRep_ActivatePlayerHearts
  (bind _hearts_gameplay_effect (Variables|GameplayEffects|GetHearts%20Gameplay%20Effect))
  (bind _outfortplayerpawns (Fortnite|GetFortPlayerPawns))
  (for _array_element _outfortplayerpawns
    (if (Variables|StageSetup|GetActivatePlayerHearts)
      (GameplayEffects|ApplyGameplayEffectToSelf :self (Ability|GetAbilitySystemComponent _array_element) :GameplayEffectClass _hearts_gameplay_effect :Level 0.0)
      (else
        (GameplayEffects|RemoveActiveGameplayEffectbySourceEffect (Ability|GetAbilitySystemComponent _array_element) _hearts_gameplay_effect 0 -1)))))

(fn OnRep_ActivatePlexus
  (bind _concert_plexus_gameplay_effect (Variables|GameplayEffects|GetConcert%20Plexus%20Gameplay%20Effect))
  (bind _outfortplayerpawns (Fortnite|GetFortPlayerPawns))
  (for _array_element _outfortplayerpawns
    (if (Variables|StageSetup|GetActivate%20Plexus)
      (Fortnite|FortShippingLog self (Utilities|String|Append "Plexus: Try to enable GamePlayEffect on " (Utilities|GetDisplayName _array_element) " using AbilitySystemComponent " (Utilities|GetDisplayName (Ability|GetAbilitySystemComponent _array_element))))
      (bind _returnvalue (GameplayEffects|ApplyGameplayEffectToSelf (Ability|GetAbilitySystemComponent _array_element) _concert_plexus_gameplay_effect 5.0 (GameplayEffects|MakeEffectContext (Ability|GetAbilitySystemComponent _array_element))))
      (bind _returnvalue_2 (Ability|GameplayEffect|GetActiveGameplayEffectStackCount _returnvalue))
      (Fortnite|FortShippingLog self (Utilities|String|Append "Plexus: " (Ability|GameplayEffect|GetActiveGameplayEffectDebugString _returnvalue) " stack count: " (Utilities|String|ToString(Integer) _returnvalue_2)))
      (else
        (GameplayEffects|RemoveActiveGameplayEffectbySourceEffect (Ability|GetAbilitySystemComponent _array_element) _concert_plexus_gameplay_effect 0 -1)))))

(fn OnRep_ActivateHologramScreen
  (Development|SetHiddeninGame (Variables|/PapayaGameplay/Blueprints/Gadgets/BP_Papaya_Concert_HologramGridPlanes.BP_Papaya_Concert_HologramGridPlanes_C|GetSR_HologramPlane (Variables|StageManagementDONOTEDIT!!|GetHologram%20Planes)) (not (Variables|StageSetup|GetActivateHologramScreen)) true))

(fn OnRep_ActivatePlayerBouncy
  (bind _bounce_gameplay_effect (Variables|GameplayEffects|GetBounce%20Gameplay%20Effect))
  (bind _lowg_gameplay_effect (Variables|GameplayEffects|GetLowG%20Gameplay%20Effect))
  (bind _outfortplayerpawns (Fortnite|GetFortPlayerPawns))
  (for _array_element _outfortplayerpawns
    (if (Variables|StageSetup|GetActivatePlayerBouncy)
      (GameplayEffects|ApplyGameplayEffectToSelf :self (Ability|GetAbilitySystemComponent _array_element) :GameplayEffectClass _bounce_gameplay_effect :Level 0.0)
      (GameplayEffects|ApplyGameplayEffectToSelf :self (Ability|GetAbilitySystemComponent _array_element) :GameplayEffectClass _lowg_gameplay_effect :Level 0.0)
      (else
        (GameplayEffects|RemoveActiveGameplayEffectbySourceEffect (Ability|GetAbilitySystemComponent _array_element) _bounce_gameplay_effect 0 -1)
        (GameplayEffects|RemoveActiveGameplayEffectbySourceEffect (Ability|GetAbilitySystemComponent _array_element) _lowg_gameplay_effect 0 -1)))))

(fn OnRep_ActivatePlayerBubbles
  (bind _bubbles_gameplay_effect (Variables|GameplayEffects|GetBubbles%20Gameplay%20Effect))
  (bind _outfortplayerpawns (Fortnite|GetFortPlayerPawns))
  (for _array_element _outfortplayerpawns
    (if (Variables|StageSetup|GetActivatePlayerBubbles)
      (GameplayEffects|ApplyGameplayEffectToSelf :self (Ability|GetAbilitySystemComponent _array_element) :GameplayEffectClass _bubbles_gameplay_effect :Level 0.0)
      (else
        (GameplayEffects|RemoveActiveGameplayEffectbySourceEffect (Ability|GetAbilitySystemComponent _array_element) _bubbles_gameplay_effect 0 -1)))))

(fn OnRep_ActivatePlayerSparkles
  (bind _sparkles_gameplay_effect (Variables|GameplayEffects|GetSparkles%20Gameplay%20Effect))
  (bind _outfortplayerpawns (Fortnite|GetFortPlayerPawns))
  (for _array_element _outfortplayerpawns
    (if (Variables|StageSetup|GetActivatePlayerSparkles)
      (GameplayEffects|ApplyGameplayEffectToSelf :self (Ability|GetAbilitySystemComponent _array_element) :GameplayEffectClass _sparkles_gameplay_effect :Level 0.0)
      (else
        (GameplayEffects|RemoveActiveGameplayEffectbySourceEffect (Ability|GetAbilitySystemComponent _array_element) _sparkles_gameplay_effect 0 -1)))))

(fn OnRep_ActivateSeqMovingLights
  (bind _moving_lights (Variables|StageManagementDONOTEDIT!!|GetMoving%20Lights))
  (bind _sequenceplayer (Variables|/Script/LevelSequence.LevelSequenceActor|GetSequencePlayer _moving_lights))
  (Networking|CanExecuteCosmeticEvents
    (:True
      (if (Variables|StageSetup|GetActivateSeqMovingLights)
        (Sequencer|Player|PlayLooping _sequenceplayer -1 :!class /Script/MovieScene.MovieSceneSequencePlayer)
        (bind _returnvalue (Sequencer|Player|GetBoundObjects _sequenceplayer (Sequencer|Player|Bindings|GetSequenceBinding 0) :!class /Script/MovieScene.MovieSceneSequencePlayer))
        (Variables|SequencerTemporary|SetActiveMovingLights _returnvalue)
        (CallFunction|FadeinMovingLights (Variables|StageSetup|GetMovingLightsInitialFadeTime))
        (CallFunction|SetMovingLightColorOffset (Variables|SequencerTemporary|GetInactive%20Color%20Offset))
        (else
          (CallFunction|FadeOutMovingLights (Variables|StageSetup|GetMovingLightsInitialFadeTime))
          (Sequencer|Player|Stop _sequenceplayer :!class /Script/MovieScene.MovieSceneSequencePlayer)
          (Utilities|Array|Clear (Variables|SequencerTemporary|GetActiveMovingLights)))))))

(fn OnRep_ActivateVODTransition
  (bind _screen_nexusrays (Variables|StageManagementDONOTEDIT!!|GetScreen%20NexusRays))
  (Networking|CanExecuteCosmeticEvents
    (:True
      (if (Variables|StageSetup|GetActivateVODTransition)
        (Class|BPScreenNexusRays|ActivateTransition _screen_nexusrays)
        (else
          (Class|BPScreenNexusRays|DeactivateTransition _screen_nexusrays))))))

(fn OnRep_AspectRatioScale
  (bind _movie_screen (Variables|StageManagementDONOTEDIT!!|GetMovie%20Screen))
  (bind _output (Utilities|ResolveSoftReference _movie_screen))
  (Networking|CanExecuteCosmeticEvents
    (:True
      (Utilities|IsValid _output
        (:Is%20Valid
          (Transformation|SetActorScale3D _output (Math|Vector|MakeVector 1.0 1.0 (Variables|StageSetup|MovieScreen|GetAspectRatioScale))))))))

(fn OnRep_BaseColorMix
  (Networking|CanExecuteCosmeticEvents
    (:True
      (Rendering|Material|SetScalarParameterValue "/Game/Packages/Fortress_SharedMaterials/GlobalMaterialParameters/FortniteMaterialParameters.FortniteMaterialParameters" "PapayaBaseColorMix" (Variables|StageSetup|Color|GetBase%20Color%20Mix) :!class /Script/Engine.KismetMaterialLibrary))))

(fn OnRep_BaseColorParameter
  (Networking|CanExecuteCosmeticEvents
    (:True
      (Rendering|Material|SetScalarParameterValue "/Game/Packages/Fortress_SharedMaterials/GlobalMaterialParameters/FortniteMaterialParameters.FortniteMaterialParameters" "PapayaBaseColorParameter" (Variables|StageSetup|Color|GetBase%20ColorParameter) :!class /Script/Engine.KismetMaterialLibrary))))

(fn OnRep_CartoonOutlinePostProcessActive
  (bind _cartoonoutlinepostprocess (Variables|StageManagementDONOTEDIT!!|GetCartoonOutlinePostProcess))
  (bind _output (Utilities|ResolveSoftReference _cartoonoutlinepostprocess))
  (Utilities|IsValid _output
    (:Is%20Valid
      (Class|PPMPostProcessStylized|SetPostEnabled _output (Variables|StageSetup|GetCartoonOutlinePostProcessActive)))))

(fn OnRep_CartoonOutlinePreset
  (bind _cartoonoutlinepostprocess (Variables|StageManagementDONOTEDIT!!|GetCartoonOutlinePostProcess))
  (bind _output (Utilities|ResolveSoftReference _cartoonoutlinepostprocess))
  (Utilities|IsValid _output
    (:Is%20Valid
      (if (== (Variables|StageSetup|GetCartoonOutlinePreset) 0)
        (bind _output_get (Variables|StageSetup|SetCartoonOutlinePostProcessActive false))
        (Class|PPMPostProcessStylized|SetPostEnabled _output _output_get)
        (elif (Variables|StageSetup|GetCartoonOutlinePostProcessActive)
          (Class|PPMPostProcessStylized|SetPostEnabled _output (stale _output_get_2))
          (switch int (Variables|StageSetup|GetCartoonOutlinePreset)
            (:Default
              _)
            (:0
              _)
            (:1
              _)
            (:2
              (Class|PPMPostProcessStylized|SetPostPreset2 _output))
            (:3
              (Class|PPMPostProcessStylized|SetPostPreset3 _output))
            (:4
              (Class|PPMPostProcessStylized|SetPostPreset4 _output))
            (:5
              (Class|PPMPostProcessStylized|SetPostPreset5 _output)))
          (Class|PPMPostProcessStylized|SetPostPreset1 _output)
          (else
            (bind _output_get_2 (Variables|StageSetup|SetCartoonOutlinePostProcessActive true))))))))

(fn OnRep_CenterEntryGlowShape
  (Networking|CanExecuteCosmeticEvents
    (:True
      (Rendering|Material|SetScalarParameterValue (Variables|/PapayaGameplay/Blueprints/Gadgets/BP_Papaya_GlowStrips.BP_Papaya_GlowStrips_C|GetCenterEntry_MID (Variables|StageManagementDONOTEDIT!!|GetEntry%20GlowStrips)) "Alpha" (Variables|StageSetup|GlowStrips|GetCenter%20Entry%20GlowShape) :!class /Script/Engine.MaterialInstanceDynamic))))

(fn OnRep_ConcertLightingActivate
  (Networking|CanExecuteCosmeticEvents
    (:True
      (CallFunction|ToggleActivateFXSystem (Variables|/PapayaGameplay/Blueprints/Gadgets/BP_Papaya_ConcertLights.BP_Papaya_ConcertLights_C|GetConcertLights (Variables|StageManagementDONOTEDIT!!|GetConcert%20Lights)) (Variables|StageSetup|GetConcert%20Lighting%20Activate)))))

(fn OnRep_ConfettiActive
  (Networking|CanExecuteCosmeticEvents
    (:True
      (CallFunction|ToggleActivateFXSystem (Variables|/Script/CascadeRuntime.Emitter|GetParticleSystemComponent (Utilities|Array|Get(acopy) (Variables|StageManagementDONOTEDIT!!|GetStage%20Particle%20Effects))) (Variables|StageSetup|Confetti|GetConfettiActive) true))))

(fn OnRep_ConfettiSpawnRate
  (Networking|CanExecuteCosmeticEvents
    (:True
      (CallFunction|SetFloatParameterOnFXSystem (Variables|/Script/CascadeRuntime.Emitter|GetParticleSystemComponent (Utilities|Array|Get(acopy) (Variables|StageManagementDONOTEDIT!!|GetStage%20Particle%20Effects))) "SpawnRateParam" (Variables|StageSetup|Confetti|GetConfettiSpawnRate)))))

(fn OnRep_Dancer1_Emote
  (bind _dancers (Variables|StageManagementDONOTEDIT!!|GetDancers))
  (bind _output (Utilities|Array|Get(acopy) _dancers))
  (Networking|CanExecuteCosmeticEvents
    (:True
      (Variables|/PapayaGameplay/Blueprints/Gadgets/Papaya_Dance_Mannequin_Master.Papaya_Dance_Mannequin_Master_C|SetEmoteSelect (Utilities|Select :Index (Variables|StageSetup|Dancer1|GetDancer1_Emote) :NewEnumerator0 0 :NewEnumerator1 1 :NewEnumerator2 2 :NewEnumerator3 3) _output)
      (Class|PapayaDanceMannequinMaster|EmotePlay _output))))

(fn OnRep_Dancer1_Mesh
  (bind _dancers (Variables|StageManagementDONOTEDIT!!|GetDancers))
  (bind _output (Utilities|Array|Get(acopy) _dancers))
  (Networking|CanExecuteCosmeticEvents
    (:True
      (Variables|/PapayaGameplay/Blueprints/Gadgets/Papaya_Dance_Mannequin_Master.Papaya_Dance_Mannequin_Master_C|SetDancer%20Select (Utilities|Select :Index (Variables|StageSetup|Dancer1|GetDancer1_Mesh) :NewEnumerator0 0 :NewEnumerator1 1 :NewEnumerator2 2 :NewEnumerator3 3 :NewEnumerator4 4 :NewEnumerator5 5 :NewEnumerator6 6 :NewEnumerator7 7 :NewEnumerator8 8 :NewEnumerator9 9 :NewEnumerator10 10) _output)
      (Class|PapayaDanceMannequinMaster|CharacterPartSetup _output))))

(fn OnRep_Dancer1_Scale
  (bind _skeletalmeshcomponents (Customization|GetAllPartSkeletalMeshComponents (Utilities|Array|Get(acopy) (Variables|StageManagementDONOTEDIT!!|GetDancers))))
  (for _array_element _skeletalmeshcomponents
    (Transformation|SetWorldScale3D _array_element (Math|Conversions|ToVector(Float) (Variables|StageSetup|Dancer1|GetDancer1_Scale)))))

(fn OnRep_Dancer1Active
  (bind _dancers (Variables|StageManagementDONOTEDIT!!|GetDancers))
  (bind _output (Utilities|Array|Get(acopy) _dancers))
  (Networking|CanExecuteCosmeticEvents
    (:True
      (if (Variables|StageSetup|Dancer1|GetDancer1Active)
        (Class|PapayaDanceMannequinMaster|ActorVisible _output)
        (else
          (Class|PapayaDanceMannequinMaster|ActorInvisible _output))))))

(fn OnRep_Dancer2_Emote
  (bind _dancers (Variables|StageManagementDONOTEDIT!!|GetDancers))
  (bind _output (Utilities|Array|Get(acopy) _dancers 1))
  (Networking|CanExecuteCosmeticEvents
    (:True
      (Variables|/PapayaGameplay/Blueprints/Gadgets/Papaya_Dance_Mannequin_Master.Papaya_Dance_Mannequin_Master_C|SetEmoteSelect (Utilities|Select :Index (Variables|StageSetup|Dancer2|GetDancer2_Emote) :NewEnumerator0 0 :NewEnumerator1 1 :NewEnumerator2 2 :NewEnumerator3 3) _output)
      (Class|PapayaDanceMannequinMaster|EmotePlay _output))))

(fn OnRep_Dancer2_Mesh
  (bind _dancers (Variables|StageManagementDONOTEDIT!!|GetDancers))
  (bind _output (Utilities|Array|Get(acopy) _dancers 1))
  (Networking|CanExecuteCosmeticEvents
    (:True
      (Variables|/PapayaGameplay/Blueprints/Gadgets/Papaya_Dance_Mannequin_Master.Papaya_Dance_Mannequin_Master_C|SetDancer%20Select (Utilities|Select :Index (Variables|StageSetup|Dancer2|GetDancer2_Mesh) :NewEnumerator0 0 :NewEnumerator1 1 :NewEnumerator2 2 :NewEnumerator3 3 :NewEnumerator4 4 :NewEnumerator5 5 :NewEnumerator6 6 :NewEnumerator7 7 :NewEnumerator8 8 :NewEnumerator9 9 :NewEnumerator10 10) _output)
      (Class|PapayaDanceMannequinMaster|CharacterPartSetup _output))))

(fn OnRep_Dancer2_Scale
  (bind _skeletalmeshcomponents (Customization|GetAllPartSkeletalMeshComponents (Utilities|Array|Get(acopy) (Variables|StageManagementDONOTEDIT!!|GetDancers) 1)))
  (for _array_element _skeletalmeshcomponents
    (Transformation|SetWorldScale3D _array_element (Math|Conversions|ToVector(Float) (Variables|StageSetup|Dancer2|GetDancer2_Scale)))))

(fn OnRep_Dancer2Active
  (bind _dancers (Variables|StageManagementDONOTEDIT!!|GetDancers))
  (bind _output (Utilities|Array|Get(acopy) _dancers 1))
  (Networking|CanExecuteCosmeticEvents
    (:True
      (if (Variables|StageSetup|Dancer2|GetDancer2Active)
        (Class|PapayaDanceMannequinMaster|ActorVisible _output)
        (else
          (Class|PapayaDanceMannequinMaster|ActorInvisible _output))))))

(fn OnRep_Dancer3_Emote
  (bind _dancers (Variables|StageManagementDONOTEDIT!!|GetDancers))
  (bind _output (Utilities|Array|Get(acopy) _dancers 2))
  (Networking|CanExecuteCosmeticEvents
    (:True
      (Variables|/PapayaGameplay/Blueprints/Gadgets/Papaya_Dance_Mannequin_Master.Papaya_Dance_Mannequin_Master_C|SetEmoteSelect (Utilities|Select :Index (Variables|StageSetup|Dancer3|GetDancer3_Emote) :NewEnumerator0 0 :NewEnumerator1 1 :NewEnumerator2 2 :NewEnumerator3 3) _output)
      (Class|PapayaDanceMannequinMaster|EmotePlay _output))))

(fn OnRep_Dancer3_Mesh
  (bind _dancers (Variables|StageManagementDONOTEDIT!!|GetDancers))
  (bind _output (Utilities|Array|Get(acopy) _dancers 2))
  (Networking|CanExecuteCosmeticEvents
    (:True
      (Variables|/PapayaGameplay/Blueprints/Gadgets/Papaya_Dance_Mannequin_Master.Papaya_Dance_Mannequin_Master_C|SetDancer%20Select (Utilities|Select :Index (Variables|StageSetup|Dancer3|GetDancer3_Mesh) :NewEnumerator0 0 :NewEnumerator1 1 :NewEnumerator2 2 :NewEnumerator3 3 :NewEnumerator4 4 :NewEnumerator5 5 :NewEnumerator6 6 :NewEnumerator7 7 :NewEnumerator8 8 :NewEnumerator9 9 :NewEnumerator10 10) _output)
      (Class|PapayaDanceMannequinMaster|CharacterPartSetup _output))))

(fn OnRep_Dancer3_Scale
  (bind _skeletalmeshcomponents (Customization|GetAllPartSkeletalMeshComponents (Utilities|Array|Get(acopy) (Variables|StageManagementDONOTEDIT!!|GetDancers) 2)))
  (for _array_element _skeletalmeshcomponents
    (Transformation|SetWorldScale3D _array_element (Math|Conversions|ToVector(Float) (Variables|StageSetup|Dancer3|GetDancer3_Scale)))))

(fn OnRep_Dancer3Active
  (bind _dancers (Variables|StageManagementDONOTEDIT!!|GetDancers))
  (bind _output (Utilities|Array|Get(acopy) _dancers 2))
  (Networking|CanExecuteCosmeticEvents
    (:True
      (if (Variables|StageSetup|Dancer3|GetDancer3Active)
        (Class|PapayaDanceMannequinMaster|ActorVisible _output)
        (else
          (Class|PapayaDanceMannequinMaster|ActorInvisible _output))))))

(fn OnRep_Dancer4_Emote
  (bind _dancers (Variables|StageManagementDONOTEDIT!!|GetDancers))
  (bind _output (Utilities|Array|Get(acopy) _dancers 3))
  (Networking|CanExecuteCosmeticEvents
    (:True
      (Variables|/PapayaGameplay/Blueprints/Gadgets/Papaya_Dance_Mannequin_Master.Papaya_Dance_Mannequin_Master_C|SetEmoteSelect (Utilities|Select :Index (Variables|StageSetup|Dancer4|GetDancer4_Emote) :NewEnumerator0 0 :NewEnumerator1 1 :NewEnumerator2 2 :NewEnumerator3 3) _output)
      (Class|PapayaDanceMannequinMaster|EmotePlay _output))))

(fn OnRep_Dancer4_Mesh
  (bind _dancers (Variables|StageManagementDONOTEDIT!!|GetDancers))
  (bind _output (Utilities|Array|Get(acopy) _dancers 3))
  (Networking|CanExecuteCosmeticEvents
    (:True
      (Variables|/PapayaGameplay/Blueprints/Gadgets/Papaya_Dance_Mannequin_Master.Papaya_Dance_Mannequin_Master_C|SetDancer%20Select (Utilities|Select :Index (Variables|StageSetup|Dancer4|GetDancer4_Mesh) :NewEnumerator0 0 :NewEnumerator1 1 :NewEnumerator2 2 :NewEnumerator3 3 :NewEnumerator4 4 :NewEnumerator5 5 :NewEnumerator6 6 :NewEnumerator7 7 :NewEnumerator8 8 :NewEnumerator9 9 :NewEnumerator10 10) _output)
      (Class|PapayaDanceMannequinMaster|CharacterPartSetup _output))))

(fn OnRep_Dancer4_Scale
  (bind _skeletalmeshcomponents (Customization|GetAllPartSkeletalMeshComponents (Utilities|Array|Get(acopy) (Variables|StageManagementDONOTEDIT!!|GetDancers) 3)))
  (for _array_element _skeletalmeshcomponents
    (Transformation|SetWorldScale3D _array_element (Math|Conversions|ToVector(Float) (Variables|StageSetup|Dancer4|GetDancer4_Scale)))))

(fn OnRep_Dancer4Active
  (bind _dancers (Variables|StageManagementDONOTEDIT!!|GetDancers))
  (bind _output (Utilities|Array|Get(acopy) _dancers 3))
  (Networking|CanExecuteCosmeticEvents
    (:True
      (if (Variables|StageSetup|Dancer4|GetDancer4Active)
        (Class|PapayaDanceMannequinMaster|ActorVisible _output)
        (else
          (Class|PapayaDanceMannequinMaster|ActorInvisible _output))))))

(fn OnRep_Desaturation
  (Networking|CanExecuteCosmeticEvents
    (:True
      (Rendering|Material|SetScalarParameterValue "/Game/Packages/Fortress_SharedMaterials/GlobalMaterialParameters/FortniteMaterialParameters.FortniteMaterialParameters" "PapayaDesaturation" (Variables|StageSetup|Color|GetDesaturation) :!class /Script/Engine.KismetMaterialLibrary))))

(fn OnRep_DiscoBall
  (bind _disco_ball (Variables|StageSetup|GetDisco%20Ball))
  (bind (_active_133_f02dd95f4b00431fc3c56a8d62335ac7 _amountofdiscoballs_165_423bf929458961569240d8b988272a90 _numberofclusters_147_b236a2f64331fdf5d39475b14e13ee7c _maxdiscoballsize_163_1cb77bc94dd2c24381318a95cb853417 _mindiscoballsize_164_aaf825b64180349048a89786cd4654e4 _rotationpivotoffsetscale_166_237121fb4a46e77d363ca7b5602863f3) (Utilities|Struct|BreakPapayaEffectsConcertDiscoBalls _disco_ball))
  (bind _discoballs (Variables|StageManagementDONOTEDIT!!|GetDiscoBalls))
  (Networking|CanExecuteCosmeticEvents
    (:True
      (for (_array_element _array_index) (Utilities|Array|MakeArray "User.BurstAmount" "User.ClusterCount" "User.ScaleMax" "User.ScaleMin" "User.PivotOffsetScale")
        (CallFunction|SetFloatParameterOnFXSystem (Variables|/PapayaGameplay/Blueprints/Gadgets/BP_Papaya_DiscoBalls.BP_Papaya_DiscoBalls_C|GetNiagara (Variables|StageManagementDONOTEDIT!!|GetDiscoBalls)) _array_element (Utilities|Array|Get(acopy) (Utilities|Array|MakeArray _amountofdiscoballs_165_423bf929458961569240d8b988272a90 _numberofclusters_147_b236a2f64331fdf5d39475b14e13ee7c _maxdiscoballsize_163_1cb77bc94dd2c24381318a95cb853417 _mindiscoballsize_164_aaf825b64180349048a89786cd4654e4 _rotationpivotoffsetscale_166_237121fb4a46e77d363ca7b5602863f3) _array_index)))
      (if _active_133_f02dd95f4b00431fc3c56a8d62335ac7
        (Class|BPPapayaDiscoBalls|Start _discoballs)
        (else
          (Class|BPPapayaDiscoBalls|Stop _discoballs))))))

(fn OnRep_DiscoLightsActive
  (Networking|CanExecuteCosmeticEvents
    (:True
      (CallFunction|ToggleActivateFXSystem (Variables|/Script/CascadeRuntime.Emitter|GetParticleSystemComponent (Utilities|Array|Get(acopy) (Variables|StageManagementDONOTEDIT!!|GetStage%20Particle%20Effects) 1)) (Variables|StageSetup|DiscoLights|GetDiscoLightsActive)))))

(fn OnRep_FireworksActive
  (Networking|CanExecuteCosmeticEvents
    (:True
      (CallFunction|ToggleActivateFXSystem (Variables|/Script/CascadeRuntime.Emitter|GetParticleSystemComponent (Utilities|Array|Get(acopy) (Variables|StageManagementDONOTEDIT!!|GetStage%20Particle%20Effects) 2)) (Variables|StageSetup|Fireworks|GetFireworksActive))
      (Default|EnableFireworksAudio (and (not (Math|Float|NearlyEqual(Float) (Variables|StageSetup|Fireworks|GetFireworksSpawnRate) 0.0 0.1)) (Variables|StageSetup|Fireworks|GetFireworksActive))))))

(fn OnRep_FireworksSpawnRate
  (Networking|CanExecuteCosmeticEvents
    (:True
      (CallFunction|SetFloatParameterOnFXSystem (Variables|/Script/CascadeRuntime.Emitter|GetParticleSystemComponent (Utilities|Array|Get(acopy) (Variables|StageManagementDONOTEDIT!!|GetStage%20Particle%20Effects) 2)) "SpawnRateParam" (Variables|StageSetup|Fireworks|GetFireworksSpawnRate))
      (Default|EnableFireworksAudio (and (not (Math|Float|NearlyEqual(Float) (Variables|StageSetup|Fireworks|GetFireworksSpawnRate) 0.0 0.1)) (Variables|StageSetup|Fireworks|GetFireworksActive))))))

(fn OnRep_FluorescentLightMode
  (bind _fluorescentlights (Variables|StageManagementDONOTEDIT!!|GetFluorescentLights))
  (bind _output (Utilities|ResolveSoftReference _fluorescentlights))
  (Utilities|IsValid _output
    (:Is%20Valid
      (if (not (Variables|StageSetup|GetFluorescentLightsActive))
        (Class|PapayaFluorescentLights|ActivateLights _output)
        (Variables|StageSetup|SetFluorescentLightsActive true))
      (switch int (Variables|StageSetup|GetFluorescentLightMode)
        (:Default
          _)
        (:0
          _)
        (:1
          (LocalVariables|SetNewPattern "NewEnumerator0")
          (Class|PapayaFluorescentLights|SetPattern _output (LocalVariables|GetNewPattern)))
        (:2
          (LocalVariables|SetNewPattern "NewEnumerator1"))
        (:3
          (LocalVariables|SetNewPattern "NewEnumerator2"))
        (:4
          (LocalVariables|SetNewPattern "NewEnumerator3"))
        (:5
          (LocalVariables|SetNewPattern "NewEnumerator4")))
      (LocalVariables|SetNewPattern "NewEnumerator6"))))

(fn OnRep_HologramPlanesAlpha
  (Networking|CanExecuteCosmeticEvents
    (:True
      (Rendering|Material|SetScalarParameterValue (Variables|/PapayaGameplay/Blueprints/Gadgets/BP_Papaya_Concert_HologramGridPlanes.BP_Papaya_Concert_HologramGridPlanes_C|GetSRHolo_MID (Variables|StageManagementDONOTEDIT!!|GetHologram%20Planes)) "Opacity" (Variables|StageSetup|GetHologramPlanesAlpha) :!class /Script/Engine.MaterialInstanceDynamic))))

(fn OnRep_IntermissionColorMix
  (Networking|CanExecuteCosmeticEvents
    (:True
      (Rendering|Material|SetScalarParameterValue "/Game/Packages/Fortress_SharedMaterials/GlobalMaterialParameters/FortniteMaterialParameters.FortniteMaterialParameters" "PapayaIntermissionColorMix" (Variables|StageSetup|Color|GetIntermission%20Color%20Mix) :!class /Script/Engine.KismetMaterialLibrary))))

(fn OnRep_IntermissionColorParameter
  (Networking|CanExecuteCosmeticEvents
    (:True
      (Rendering|Material|SetScalarParameterValue "/Game/Packages/Fortress_SharedMaterials/GlobalMaterialParameters/FortniteMaterialParameters.FortniteMaterialParameters" "PapayaIntermissionColorParameter" (Variables|StageSetup|Color|GetIntermission%20Color%20Parameter) :!class /Script/Engine.KismetMaterialLibrary))))

(fn OnRep_KaleidoscopeValue
  (Rendering|Material|SetScalarParameterValue "/Game/Environments/Papaya/Materials/MPC_PapayaGadgetControls.MPC_PapayaGadgetControls" "ScreenKaleidoscope" (Variables|StageSetup|GetKaleidoscopeValue) :!class /Script/Engine.KismetMaterialLibrary))

(fn OnRep_Laser1
  (Networking|CanExecuteCosmeticEvents
    (:True
      (CallFunction|ToggleActivateFXSystem (Variables|/PapayaGameplay/Blueprints/Gadgets/BP_Papaya_Concert_Laser.BP_Papaya_Concert_Laser_C|GetNiagara (Utilities|Array|Get(acopy) (Variables|StageManagementDONOTEDIT!!|GetLasers))) (Variables|StageSetup|Lasers|GetLaser1)))))

(fn OnRep_Laser1_Color
  (Networking|CanExecuteCosmeticEvents
    (:True
      (Niagara|SetNiagaraVariable(LinearColor) (Variables|/PapayaGameplay/Blueprints/Gadgets/BP_Papaya_Concert_Laser.BP_Papaya_Concert_Laser_C|GetNiagara (Utilities|Array|Get(acopy) (Variables|StageManagementDONOTEDIT!!|GetLasers))) "LaserColor" (Variables|StageSetup|Lasers|GetLaser1_Color)))))

(fn OnRep_Laser1_CustomColorUse
  (Networking|CanExecuteCosmeticEvents
    (:True
      (CallFunction|SetFloatParameterOnFXSystem (Variables|/PapayaGameplay/Blueprints/Gadgets/BP_Papaya_Concert_Laser.BP_Papaya_Concert_Laser_C|GetNiagara (Utilities|Array|Get(acopy) (Variables|StageManagementDONOTEDIT!!|GetLasers))) "UseCustomColor" (Variables|StageSetup|Lasers|GetLaser1_CustomColorUse)))))

(fn OnRep_Laser2
  (Networking|CanExecuteCosmeticEvents
    (:True
      (CallFunction|ToggleActivateFXSystem (Variables|/PapayaGameplay/Blueprints/Gadgets/BP_Papaya_Concert_Laser.BP_Papaya_Concert_Laser_C|GetNiagara (Utilities|Array|Get(acopy) (Variables|StageManagementDONOTEDIT!!|GetLasers) 1)) (Variables|StageSetup|Lasers|GetLaser2)))))

(fn OnRep_Laser2_Color
  (Networking|CanExecuteCosmeticEvents
    (:True
      (Niagara|SetNiagaraVariable(LinearColor) (Variables|/PapayaGameplay/Blueprints/Gadgets/BP_Papaya_Concert_Laser.BP_Papaya_Concert_Laser_C|GetNiagara (Utilities|Array|Get(acopy) (Variables|StageManagementDONOTEDIT!!|GetLasers) 1)) "LaserColor" (Variables|StageSetup|Lasers|GetLaser2_Color)))))

(fn OnRep_Laser2_CustomColorUse
  (Networking|CanExecuteCosmeticEvents
    (:True
      (CallFunction|SetFloatParameterOnFXSystem (Variables|/PapayaGameplay/Blueprints/Gadgets/BP_Papaya_Concert_Laser.BP_Papaya_Concert_Laser_C|GetNiagara (Utilities|Array|Get(acopy) (Variables|StageManagementDONOTEDIT!!|GetLasers) 1)) "UseCustomColor" (Variables|StageSetup|Lasers|GetLaser2_CustomColorUse)))))

(fn OnRep_Laser3
  (Networking|CanExecuteCosmeticEvents
    (:True
      (CallFunction|ToggleActivateFXSystem (Variables|/PapayaGameplay/Blueprints/Gadgets/BP_Papaya_Concert_Laser.BP_Papaya_Concert_Laser_C|GetNiagara (Utilities|Array|Get(acopy) (Variables|StageManagementDONOTEDIT!!|GetLasers) 2)) (Variables|StageSetup|Lasers|GetLaser3)))))

(fn OnRep_Laser3_Color
  (Networking|CanExecuteCosmeticEvents
    (:True
      (Niagara|SetNiagaraVariable(LinearColor) (Variables|/PapayaGameplay/Blueprints/Gadgets/BP_Papaya_Concert_Laser.BP_Papaya_Concert_Laser_C|GetNiagara (Utilities|Array|Get(acopy) (Variables|StageManagementDONOTEDIT!!|GetLasers) 2)) "LaserColor" (Variables|StageSetup|Lasers|GetLaser3_Color)))))

(fn OnRep_Laser3_CustomColorUse
  (Networking|CanExecuteCosmeticEvents
    (:True
      (CallFunction|SetFloatParameterOnFXSystem (Variables|/PapayaGameplay/Blueprints/Gadgets/BP_Papaya_Concert_Laser.BP_Papaya_Concert_Laser_C|GetNiagara (Utilities|Array|Get(acopy) (Variables|StageManagementDONOTEDIT!!|GetLasers) 2)) "UseCustomColor" (Variables|StageSetup|Lasers|GetLaser3_CustomColorUse)))))

(fn OnRep_Laser4
  (Networking|CanExecuteCosmeticEvents
    (:True
      (CallFunction|ToggleActivateFXSystem (Variables|/PapayaGameplay/Blueprints/Gadgets/BP_Papaya_Concert_Laser.BP_Papaya_Concert_Laser_C|GetNiagara (Utilities|Array|Get(acopy) (Variables|StageManagementDONOTEDIT!!|GetLasers) 3)) (Variables|StageSetup|Lasers|GetLaser4)))))

(fn OnRep_Laser4_Color
  (Networking|CanExecuteCosmeticEvents
    (:True
      (Niagara|SetNiagaraVariable(LinearColor) (Variables|/PapayaGameplay/Blueprints/Gadgets/BP_Papaya_Concert_Laser.BP_Papaya_Concert_Laser_C|GetNiagara (Utilities|Array|Get(acopy) (Variables|StageManagementDONOTEDIT!!|GetLasers) 3)) "LaserColor" (Variables|StageSetup|Lasers|GetLaser4_Color)))))

(fn OnRep_Laser4_CustomColorUse
  (Networking|CanExecuteCosmeticEvents
    (:True
      (CallFunction|SetFloatParameterOnFXSystem (Variables|/PapayaGameplay/Blueprints/Gadgets/BP_Papaya_Concert_Laser.BP_Papaya_Concert_Laser_C|GetNiagara (Utilities|Array|Get(acopy) (Variables|StageManagementDONOTEDIT!!|GetLasers) 3)) "UseCustomColor" (Variables|StageSetup|Lasers|GetLaser4_CustomColorUse)))))

(fn OnRep_Laser5
  (Networking|CanExecuteCosmeticEvents
    (:True
      (CallFunction|ToggleActivateFXSystem (Variables|/PapayaGameplay/Blueprints/Gadgets/BP_Papaya_Concert_Laser.BP_Papaya_Concert_Laser_C|GetNiagara (Utilities|Array|Get(acopy) (Variables|StageManagementDONOTEDIT!!|GetLasers) 4)) (Variables|StageSetup|Lasers|GetLaser5)))))

(fn OnRep_Laser5_Color
  (Networking|CanExecuteCosmeticEvents
    (:True
      (Niagara|SetNiagaraVariable(LinearColor) (Variables|/PapayaGameplay/Blueprints/Gadgets/BP_Papaya_Concert_Laser.BP_Papaya_Concert_Laser_C|GetNiagara (Utilities|Array|Get(acopy) (Variables|StageManagementDONOTEDIT!!|GetLasers) 4)) "LaserColor" (Variables|StageSetup|Lasers|GetLaser5_Color)))))

(fn OnRep_Laser5_CustomColorUse
  (Networking|CanExecuteCosmeticEvents
    (:True
      (CallFunction|SetFloatParameterOnFXSystem (Variables|/PapayaGameplay/Blueprints/Gadgets/BP_Papaya_Concert_Laser.BP_Papaya_Concert_Laser_C|GetNiagara (Utilities|Array|Get(acopy) (Variables|StageManagementDONOTEDIT!!|GetLasers) 4)) "UseCustomColor" (Variables|StageSetup|Lasers|GetLaser5_CustomColorUse)))))

(fn OnRep_Laser6
  (Networking|CanExecuteCosmeticEvents
    (:True
      (CallFunction|ToggleActivateFXSystem (Variables|/PapayaGameplay/Blueprints/Gadgets/BP_Papaya_Concert_Laser.BP_Papaya_Concert_Laser_C|GetNiagara (Utilities|Array|Get(acopy) (Variables|StageManagementDONOTEDIT!!|GetLasers) 5)) (Variables|StageSetup|Lasers|GetLaser6)))))

(fn OnRep_Laser6_Color
  (Networking|CanExecuteCosmeticEvents
    (:True
      (Niagara|SetNiagaraVariable(LinearColor) (Variables|/PapayaGameplay/Blueprints/Gadgets/BP_Papaya_Concert_Laser.BP_Papaya_Concert_Laser_C|GetNiagara (Utilities|Array|Get(acopy) (Variables|StageManagementDONOTEDIT!!|GetLasers) 5)) "LaserColor" (Variables|StageSetup|Lasers|GetLaser6_Color)))))

(fn OnRep_Laser6_CustomColorUse
  (Networking|CanExecuteCosmeticEvents
    (:True
      (CallFunction|SetFloatParameterOnFXSystem (Variables|/PapayaGameplay/Blueprints/Gadgets/BP_Papaya_Concert_Laser.BP_Papaya_Concert_Laser_C|GetNiagara (Utilities|Array|Get(acopy) (Variables|StageManagementDONOTEDIT!!|GetLasers) 5)) "UseCustomColor" (Variables|StageSetup|Lasers|GetLaser6_CustomColorUse)))))

(fn OnRep_LaserSensitivity
  (Networking|CanExecuteCosmeticEvents
    (:True
      (for _array_element (Variables|StageManagementDONOTEDIT!!|GetLasers)
        (Utilities|IsValid _array_element
          (:Is%20Valid
            (CallFunction|SetFloatParameterOnFXSystem (Variables|/PapayaGameplay/Blueprints/Gadgets/BP_Papaya_Concert_Laser.BP_Papaya_Concert_Laser_C|GetNiagara _array_element) "User.LaserEnv" (Variables|StageSetup|Lasers|GetLaserSensitivity))))))))

(fn OnRep_MainScreenActive
  (CallFunction|MainScreenActiveChanged))

(fn OnRep_MainScreenAlpha
  (bind _mainscreenalpha (Variables|StageSetup|GetMainScreenAlpha))
  (bind _returnvalue (Math|Vector|Lerp(Vector) "1.000000,1.000000,1.000000" "0, 0, 0" _mainscreenalpha))
  (bind _mainscreenalpha_2 (Variables|StageSetup|GetMainScreenAlpha))
  (bind _returnvalue_2 (Math|Float|Lerp 1.0 0.0 _mainscreenalpha_2))
  (bind _returnvalue_3 (Math|Vector|Lerp(Vector) "-1024.000000,26575.000000,-2648.000000" "-1024.000000,31575.000000,2848.000000" _mainscreenalpha))
  (bind _screen_actor (Variables|StageManagementDONOTEDIT!!|GetScreen%20Actor))
  (bind _output (Utilities|ResolveSoftReference _screen_actor))
  (bind _side_screens_actor (Variables|StageManagementDONOTEDIT!!|GetSide%20Screens%20Actor))
  (bind _output_2 (Utilities|ResolveSoftReference _side_screens_actor))
  (bind _floating_large_shards (Variables|StageManagementDONOTEDIT!!|GetFloating%20Large%20Shards))
  (bind _output_3 (Utilities|ResolveSoftReference _floating_large_shards))
  (bind _staticmeshcomponent (Variables|/Script/Engine.StaticMeshActor|GetStaticMeshComponent _output))
  (bind _screen_nexusrays (Variables|StageManagementDONOTEDIT!!|GetScreen%20NexusRays))
  (bind _staticmeshcomponent_2 (Variables|/Script/Engine.StaticMeshActor|GetStaticMeshComponent _output_2))
  (Networking|CanExecuteCosmeticEvents
    (:True
      (Utilities|FlowControl|Sequence
        (:then_0
          (Utilities|IsValid _output
            (:Is%20Valid
              (Transformation|SetWorldScale3D _staticmeshcomponent _returnvalue)
              (Transformation|SetWorldLocation _staticmeshcomponent _returnvalue_3 false false))
            (:Is%20Not%20Valid
              _))
          (Utilities|IsValid _output_2
            (:Is%20Valid
              (Transformation|SetWorldScale3D _staticmeshcomponent_2 _returnvalue)
              (Transformation|SetWorldLocation _staticmeshcomponent_2 _returnvalue_3 false false))
            (:Is%20Not%20Valid
              _))
          (Utilities|IsValid _output_3
            (:Is%20Valid
              (Transformation|SetActorScale3D _output_3 _returnvalue)
              (Transformation|SetActorLocation _output_3 _returnvalue_3 false false)
              (Transformation|SetActorScale3D _screen_nexusrays _returnvalue)
              (Transformation|SetActorLocation _screen_nexusrays _returnvalue_3 false false))))
        (:then_1
          (Variables|StageSetup|GlowStrips|SetScreen%20Curve%201%20Alpha _returnvalue_2)
          (Variables|StageSetup|GlowStrips|SetScreen%20Curve%202%20Alpha _returnvalue_2)
          (Variables|StageSetup|GlowStrips|SetScreen%20Curve%203%20Alpha _returnvalue_2)
          (Variables|StageSetup|GlowStrips|SetScreen%20Curve%204%20Alpha _returnvalue_2)
          (Variables|StageSetup|SetScreen%20Nexus%20Rays%20Opacity _returnvalue_2)
          (Variables|StageSetup|SetScreen%20Haze%20Opacity _returnvalue_2))))))

(fn OnRep_MoviePostBlendWeight
  (bind _movie_screen_post (Variables|StageManagementDONOTEDIT!!|GetMovie%20Screen%20Post))
  (bind _output (Utilities|ResolveSoftReference _movie_screen_post))
  (bind _moviepostblendweight (Variables|StageSetup|MovieScreen|GetMoviePostBlendWeight))
  (Networking|CanExecuteCosmeticEvents
    (:True
      (Utilities|IsValid _output
        (:Is%20Valid
          (Variables|/Script/Engine.PostProcessVolume|SetBlendWeight _moviepostblendweight _output))
        (:Is%20Not%20Valid
          _))
      (Rendering|Material|SetScalarParameterValue (Variables|/PapayaGameplay/Blueprints/Gadgets/BP_Papaya_Movie_DarkeningPlane.BP_Papaya_Movie_DarkeningPlane_C|GetDarkeningPlaneMID (Variables|StageManagementDONOTEDIT!!|GetMovieDarkeningPlane)) "Opacity" _moviepostblendweight :!class /Script/Engine.MaterialInstanceDynamic))))

(fn OnRep_MovingLightColorOffset
  (bind _activemovinglights (Variables|SequencerTemporary|GetActiveMovingLights))
  (Networking|CanExecuteCosmeticEvents
    (:True
      (if (Utilities|Array|IsValidIndex _activemovinglights)
        (for _array_element _activemovinglights
          (bind _asbp_stage_light_master (Utilities|Casting|CastToBP_StageLIghtMaster _array_element))
          (Utilities|IsValid _asbp_stage_light_master
            (:Is%20Valid
              (Class|BPStageLIghtMaster|SetColorOffset _asbp_stage_light_master (Variables|StageSetup|GetMovingLightColorOffset) :!class /PapayaGameplay/Blueprints/Gadgets/StageLighting/BP_StageLIghtMaster.BP_StageLIghtMaster_C))))
        (else
          (Variables|SequencerTemporary|SetInactive%20Color%20Offset 0.0))))))

(fn OnRep_MovingLightsAlpha
  (Networking|CanExecuteCosmeticEvents
    (:True
      (for _array_element (Variables|SequencerTemporary|GetActiveMovingLights)
        (Utilities|IsValid _array_element
          (:Is%20Valid
            (bind _asbp_stage_light_master (Utilities|Casting|CastToBP_StageLIghtMaster _array_element))
            (Class|BPStageLIghtMaster|Fade _asbp_stage_light_master (Variables|StageSetup|GetMovingLightsAlpha)))
          (:Is%20Not%20Valid
            (Development|PrintString "Active Moving Lights INVALID")))))))

(fn OnRep_OchoScreenActive
  (bind _papayapostercontrol (Variables|IslandScreensDONOTEDIT!!|GetPapayaPosterControl))
  (bind _output (Utilities|ResolveSoftReference _papayapostercontrol))
  (Utilities|IsValid _output
    (:Is%20Valid
      (Class|BPPapayaPosterControl|EnableSpecialEventPoster _output (Variables|IslandScreensDONOTEDIT!!|GetOchoScreenActive)))))

(fn OnRep_RedirectToMovieScreenActive
  (bind _papayascreen02_music (Variables|IslandScreensDONOTEDIT!!|GetPapayaScreen02_Music))
  (bind _output (Utilities|ResolveSoftReference _papayascreen02_music))
  (Utilities|IsValid _output
    (:Is%20Valid
      (Class|BPPapayaScreen02|EnableRedirectMaptoMovie _output true))))

(fn OnRep_RedirectToMusicScreenActive
  (bind _papayascreen01_movie (Variables|IslandScreensDONOTEDIT!!|GetPapayaScreen01_Movie))
  (bind _output (Utilities|ResolveSoftReference _papayascreen01_movie))
  (Utilities|IsValid _output
    (:Is%20Valid
      (Class|BPPapayaScreen01|EnableRedirectMaptoMusic _output (Variables|IslandScreensDONOTEDIT!!|GetRedirectToMusicScreenActive)))))

(fn OnRep_ScreenCurve1Alpha
  (Networking|CanExecuteCosmeticEvents
    (:True
      (Rendering|Material|SetScalarParameterValue (Variables|/PapayaGameplay/Blueprints/Gadgets/BP_Papaya_GlowStrips_Screen.BP_Papaya_GlowStrips_Screen_C|GetScreenCurve1_MID (Variables|StageManagementDONOTEDIT!!|GetScreen%20GlowStrips)) "Alpha" (Variables|StageSetup|GlowStrips|GetScreen%20Curve%201%20Alpha) :!class /Script/Engine.MaterialInstanceDynamic))))

(fn OnRep_ScreenCurve2Alpha
  (Networking|CanExecuteCosmeticEvents
    (:True
      (Rendering|Material|SetScalarParameterValue (Variables|/PapayaGameplay/Blueprints/Gadgets/BP_Papaya_GlowStrips_Screen.BP_Papaya_GlowStrips_Screen_C|GetScreenCurve2_MID (Variables|StageManagementDONOTEDIT!!|GetScreen%20GlowStrips)) "Alpha" (Variables|StageSetup|GlowStrips|GetScreen%20Curve%202%20Alpha) :!class /Script/Engine.MaterialInstanceDynamic))))

(fn OnRep_ScreenCurve3Alpha
  (Networking|CanExecuteCosmeticEvents
    (:True
      (Rendering|Material|SetScalarParameterValue (Variables|/PapayaGameplay/Blueprints/Gadgets/BP_Papaya_GlowStrips_Screen.BP_Papaya_GlowStrips_Screen_C|GetScreenCurve3_MID (Variables|StageManagementDONOTEDIT!!|GetScreen%20GlowStrips)) "Alpha" (Variables|StageSetup|GlowStrips|GetScreen%20Curve%203%20Alpha) :!class /Script/Engine.MaterialInstanceDynamic))))

(fn OnRep_ScreenCurve4Alpha
  (Networking|CanExecuteCosmeticEvents
    (:True
      (Rendering|Material|SetScalarParameterValue (Variables|/PapayaGameplay/Blueprints/Gadgets/BP_Papaya_GlowStrips_Screen.BP_Papaya_GlowStrips_Screen_C|GetScreenCurve4_MID (Variables|StageManagementDONOTEDIT!!|GetScreen%20GlowStrips)) "Alpha" (Variables|StageSetup|GlowStrips|GetScreen%20Curve%204%20Alpha) :!class /Script/Engine.MaterialInstanceDynamic))))

(fn OnRep_ScreenHazeOpacity
  (Networking|CanExecuteCosmeticEvents
    (:True
      (Rendering|Material|SetScalarParameterValue (Variables|/PapayaGameplay/Blueprints/Gadgets/BP_ShardScreenStreams.BP_ShardScreenStreams_C|GetHaze_MID (Variables|StageManagementDONOTEDIT!!|GetScreen%20ShardStreams%20Haze)) "Opacity" (Variables|StageSetup|GetScreen%20Haze%20Opacity) :!class /Script/Engine.MaterialInstanceDynamic))))

(fn OnRep_ScreenNexusActivate
  (Networking|CanExecuteCosmeticEvents
    (:True
      (for _array_element (Utilities|Array|MakeArray (Variables|/PapayaGameplay/Blueprints/Gadgets/BP_Screen_NexusRays.BP_Screen_NexusRays_C|GetP_Papaya_CattusDoggus_Destabilize1 (Variables|StageManagementDONOTEDIT!!|GetScreen%20NexusRays)) (Variables|/PapayaGameplay/Blueprints/Gadgets/BP_Screen_NexusRays.BP_Screen_NexusRays_C|GetP_Papaya_CattusDoggus_Destabilize2 (Variables|StageManagementDONOTEDIT!!|GetScreen%20NexusRays)) (Variables|/PapayaGameplay/Blueprints/Gadgets/BP_Screen_NexusRays.BP_Screen_NexusRays_C|GetP_Papaya_CattusDoggus_Destabilize3 (Variables|StageManagementDONOTEDIT!!|GetScreen%20NexusRays)) (Variables|/PapayaGameplay/Blueprints/Gadgets/BP_Screen_NexusRays.BP_Screen_NexusRays_C|GetP_Papaya_CattusDoggus_Destabilize4 (Variables|StageManagementDONOTEDIT!!|GetScreen%20NexusRays)))
        (CallFunction|ToggleActivateFXSystem _array_element (Variables|StageSetup|GetScreen%20Nexus%20Activate))))))

(fn OnRep_ScreenNexusRaysOpacity
  (bind _screen_nexusrays (Variables|StageManagementDONOTEDIT!!|GetScreen%20NexusRays))
  (bind _screen_nexus_rays_opacity (Variables|StageSetup|GetScreen%20Nexus%20Rays%20Opacity))
  (Networking|CanExecuteCosmeticEvents
    (:True
      (Rendering|Material|SetScalarParameterValue (Variables|/PapayaGameplay/Blueprints/Gadgets/BP_Screen_NexusRays.BP_Screen_NexusRays_C|GetNexusRays_MID _screen_nexusrays) "Alpha" _screen_nexus_rays_opacity :!class /Script/Engine.MaterialInstanceDynamic)
      (Rendering|Material|SetScalarParameterValue (Variables|/PapayaGameplay/Blueprints/Gadgets/BP_Screen_NexusRays.BP_Screen_NexusRays_C|GetNexusRaysSide_MID _screen_nexusrays) "Alpha" _screen_nexus_rays_opacity :!class /Script/Engine.MaterialInstanceDynamic))))

(fn OnRep_ScreenSideEqualizersActivate
  (Networking|CanExecuteCosmeticEvents
    (:True
      (CallFunction|ToggleActorHiddenInGame (Variables|StageSetup|GetScreen%20Side%20Equalizers%20Activate) (Variables|StageManagementDONOTEDIT!!|GetScreen%20Side%20EQs)))))

(fn OnRep_ScreenShards
  (bind _floating_large_shards (Variables|StageManagementDONOTEDIT!!|GetFloating%20Large%20Shards))
  (bind _output (Utilities|ResolveSoftReference _floating_large_shards))
  (bind _screenshards (Variables|Default|GetScreenShards))
  (Networking|CanExecuteCosmeticEvents
    (:True
      (Components|Activation|SetActive (Variables|/PapayaGameplay/Blueprints/Gadgets/BP_ShardScreenStreams.BP_ShardScreenStreams_C|GetNS_ShardScreenStream1 (Variables|StageManagementDONOTEDIT!!|GetScreen%20ShardStreams%20Haze)) _screenshards false)
      (if _screenshards
        (Class|BPPapayaScreenShards|ShowMainScreenShard _output)
        (else
          (Class|BPPapayaScreenShards|HideMainScreenShard _output))))))

(fn OnRep_ScreenShardsRate
  (Networking|CanExecuteCosmeticEvents
    (:True
      (Effects|Components|ParticleSystem|SetFloatParameter (Variables|/PapayaGameplay/Blueprints/Gadgets/BP_ShardScreenStreams.BP_ShardScreenStreams_C|GetNS_ShardScreenStream1 (Variables|StageManagementDONOTEDIT!!|GetScreen%20ShardStreams%20Haze)) "SpawnRate" (Variables|Default|GetScreenShardsRate)))))

(fn OnRep_SLEntryGlowShape
  (Networking|CanExecuteCosmeticEvents
    (:True
      (Rendering|Material|SetScalarParameterValue (Variables|/PapayaGameplay/Blueprints/Gadgets/BP_Papaya_GlowStrips.BP_Papaya_GlowStrips_C|GetSLEntry_MID (Variables|StageManagementDONOTEDIT!!|GetEntry%20GlowStrips)) "Alpha" (Variables|StageSetup|GlowStrips|GetSL%20Entry%20GlowShape) :!class /Script/Engine.MaterialInstanceDynamic))))

(fn OnRep_SLTetraScreen
  (bind _tetrascreenactors (Variables|StageManagementDONOTEDIT!!|GetTetraScreenActors))
  (bind _output (Utilities|Array|Get(acopy) _tetrascreenactors 1))
  (bind _tetrascreenactors_2 (Variables|StageManagementDONOTEDIT!!|GetTetraScreenActors))
  (bind _output_2 (Utilities|Array|Get(acopy) _tetrascreenactors_2 1))
  (Networking|CanExecuteCosmeticEvents
    (:True
      (Class|BPPapayaTetraScreen|Hidden (Utilities|Array|Get(acopy) (Variables|StageManagementDONOTEDIT!!|GetTetraScreenActors) 1) (not (Utilities|Struct|BreakPapayaEffectsConcertTetraScreen (Variables|StageSetup|GetSLTetraScreen))))
      (Variables|/PapayaGameplay/Blueprints/Gadgets/BP_Papaya_TetraScreen.BP_Papaya_TetraScreen_C|SetRotation%20Speed (Utilities|Struct|BreakPapayaEffectsConcertTetraScreen (Variables|StageSetup|GetSLTetraScreen)) _output_2)
      (Class|BPPapayaTetraScreen|RotateSpeed _output_2)
      (if (Utilities|Struct|BreakPapayaEffectsConcertTetraScreen (Variables|StageSetup|GetSLTetraScreen))
        (Class|BPPapayaTetraScreen|StartRotate _output)
        (else
          (Class|BPPapayaTetraScreen|EndRotate _output))))))

(fn OnRep_SparksRainActive
  (Networking|CanExecuteCosmeticEvents
    (:True
      (CallFunction|ToggleActivateFXSystem (Variables|/Script/CascadeRuntime.Emitter|GetParticleSystemComponent (Utilities|Array|Get(acopy) (Variables|StageManagementDONOTEDIT!!|GetStage%20Particle%20Effects) 3)) (Variables|StageSetup|SparksRain|GetSparksRainActive)))))

(fn OnRep_SparksRainSpawnRate
  (Networking|CanExecuteCosmeticEvents
    (:True
      (CallFunction|SetFloatParameterOnFXSystem (Variables|/Script/CascadeRuntime.Emitter|GetParticleSystemComponent (Utilities|Array|Get(acopy) (Variables|StageManagementDONOTEDIT!!|GetStage%20Particle%20Effects) 3)) "SpawnRateParam" (Variables|StageSetup|SparksRain|GetSparksRainSpawnRate)))))

(fn OnRep_SpecialEventPoster2Active
  (bind _papayapostercontrol (Variables|IslandScreensDONOTEDIT!!|GetPapayaPosterControl))
  (bind _output (Utilities|ResolveSoftReference _papayapostercontrol))
  (Utilities|IsValid _output
    (:Is%20Valid
      (Class|BPPapayaPosterControl|EnableSpecialEventPoster2 _output (Variables|IslandScreensDONOTEDIT!!|GetSpecialEventPoster2Active)))))

(fn OnRep_SpriteDomeActive
  (bind _spritedome (Variables|StageManagementDONOTEDIT!!|GetSpriteDome))
  (bind _output (Utilities|ResolveSoftReference _spritedome))
  (Utilities|IsValid _output
    (:Is%20Valid
      (Class|BPPapayaSpriteDome|ShowDome _output (Variables|StageSetup|GetSpriteDomeActive)))))

(fn OnRep_SREntryGlowShape
  (Networking|CanExecuteCosmeticEvents
    (:True
      (Rendering|Material|SetScalarParameterValue (Variables|/PapayaGameplay/Blueprints/Gadgets/BP_Papaya_GlowStrips.BP_Papaya_GlowStrips_C|GetSREntry_MID (Variables|StageManagementDONOTEDIT!!|GetEntry%20GlowStrips)) "Alpha" (Variables|StageSetup|GlowStrips|GetSR%20Entry%20GlowShape) :!class /Script/Engine.MaterialInstanceDynamic))))

(fn OnRep_SRTetraScreen
  (bind _tetrascreenactors (Variables|StageManagementDONOTEDIT!!|GetTetraScreenActors))
  (bind _output (Utilities|Array|Get(acopy) _tetrascreenactors))
  (bind _tetrascreenactors_2 (Variables|StageManagementDONOTEDIT!!|GetTetraScreenActors))
  (bind _output_2 (Utilities|Array|Get(acopy) _tetrascreenactors_2))
  (Networking|CanExecuteCosmeticEvents
    (:True
      (Class|BPPapayaTetraScreen|Hidden (Utilities|Array|Get(acopy) (Variables|StageManagementDONOTEDIT!!|GetTetraScreenActors)) (not (Utilities|Struct|BreakPapayaEffectsConcertTetraScreen (Variables|StageSetup|GetSR%20TetraScreen))))
      (Variables|/PapayaGameplay/Blueprints/Gadgets/BP_Papaya_TetraScreen.BP_Papaya_TetraScreen_C|SetRotation%20Speed (Utilities|Struct|BreakPapayaEffectsConcertTetraScreen (Variables|StageSetup|GetSR%20TetraScreen)) _output_2)
      (Class|BPPapayaTetraScreen|RotateSpeed _output_2)
      (if (Utilities|Struct|BreakPapayaEffectsConcertTetraScreen (Variables|StageSetup|GetSR%20TetraScreen))
        (Class|BPPapayaTetraScreen|StartRotate _output)
        (else
          (Class|BPPapayaTetraScreen|EndRotate _output))))))

(fn OnRep_TetraTint
  (Networking|CanExecuteCosmeticEvents
    (:True
      (Rendering|Material|SetScalarParameterValue "/Game/Environments/Papaya/Materials/MPC_PapayaGadgetControls.MPC_PapayaGadgetControls" "TetraScreenTintMix" (Variables|StageSetup|GetTetraTint) :!class /Script/Engine.KismetMaterialLibrary))))

(fn OnRep_TetraUV
  (Networking|CanExecuteCosmeticEvents
    (:True
      (Rendering|Material|SetScalarParameterValue "/Game/Environments/Papaya/Materials/MPC_PapayaGadgetControls.MPC_PapayaGadgetControls" "ScreensUseQuadrants" (Variables|StageSetup|GetTetraUV) :!class /Script/Engine.KismetMaterialLibrary))))

(fn OnRep_TVScreenMode
  (for _array_element (Variables|StageManagementDONOTEDIT!!|GetScreenFeedTVs)
    (Utilities|IsValid (Utilities|ResolveSoftReference _array_element)
      (:Is%20Valid
        (switch int (Variables|StageSetup|GetTVScreenMode)
          (:Default
            _)
          (:0
            _)
          (:1
            (Class|BPPapayaTV|EnableTVColorSync (Utilities|ResolveSoftReference _array_element)))
          (:2
            (Class|BPPapayaTV|EnableTVMusicStage (Utilities|ResolveSoftReference _array_element)))
          (:3
            (Class|BPPapayaTV|EnableTVMovieScreen (Utilities|ResolveSoftReference _array_element))))
        (Class|BPPapayaTV|DisableTV (Utilities|ResolveSoftReference _array_element))))))

(fn OnRep_WorldGlobeVisible
  (bind _worldglobe (Variables|StageManagementDONOTEDIT!!|GetWorldGlobe))
  (bind _output (Utilities|ResolveSoftReference _worldglobe))
  (if (Variables|StageSetup|GetWorldGlobeVisible)
    (Class|BPPapayaGlobe|ToggleGlobeOn _output)
    (else
      (Class|BPPapayaGlobe|ToggleGlobeOff _output))))

(fn SetFloatParameteronFXSystem
  (params (FX_Asset :type FXSystem%20Component%20Object%20Reference) (Parameter%20Name :type Name) (Value :type Float%20%28double-precision%29))
  (Utilities|IsValid FX_Asset
    (:Is%20Valid
      (Effects|Components|ParticleSystem|SetFloatParameter FX_Asset Parameter%20Name Value))))

(fn SpectralAnalysisUpdate
  (bind _returnvalue (Niagara|GetNiagaraParameterCollection "/Game/Athena/Sounds/Events/Papaya/FX/Papaya_Jukebox_NPC.Papaya_Jukebox_NPC"))
  (bind _returnvalue_2 (Rendering|Material|GetScalarParameterValue "/Game/Packages/Fortress_SharedMaterials/GlobalMaterialParameters/FortniteMaterialParameters.FortniteMaterialParameters" (Variables|MPCNames|GetBaseColorParameter) :!class /Script/Engine.KismetMaterialLibrary))
  (Niagara|SetFloatParameter _returnvalue "Base Color Parameter" (Math|Float|Fraction _returnvalue_2))
  (bind _returnvalue_3 (Rendering|Material|GetScalarParameterValue "/Game/Packages/Fortress_SharedMaterials/GlobalMaterialParameters/FortniteMaterialParameters.FortniteMaterialParameters" (Variables|MPCNames|GetBaseColorMix) :!class /Script/Engine.KismetMaterialLibrary))
  (Niagara|SetFloatParameter _returnvalue "BaseColorMix" _returnvalue_3)
  (bind _returnvalue_4 (Rendering|Material|GetScalarParameterValue "/Game/Packages/Fortress_SharedMaterials/GlobalMaterialParameters/FortniteMaterialParameters.FortniteMaterialParameters" (Variables|MPCNames|GetIntermissionColorParameter) :!class /Script/Engine.KismetMaterialLibrary))
  (Niagara|SetFloatParameter _returnvalue "Intermission Color Parameter" _returnvalue_4)
  (bind _returnvalue_5 (Rendering|Material|GetScalarParameterValue "/Game/Packages/Fortress_SharedMaterials/GlobalMaterialParameters/FortniteMaterialParameters.FortniteMaterialParameters" (Variables|MPCNames|GetIntermissionColorMix) :!class /Script/Engine.KismetMaterialLibrary))
  (Niagara|SetFloatParameter _returnvalue "Intermission Color Mix" _returnvalue_5)
  (bind _returnvalue_6 (Rendering|Material|GetScalarParameterValue "/Game/Packages/Fortress_SharedMaterials/GlobalMaterialParameters/FortniteMaterialParameters.FortniteMaterialParameters" (Variables|MPCNames|GetDesatName) :!class /Script/Engine.KismetMaterialLibrary))
  (Niagara|SetFloatParameter _returnvalue "Desaturation" _returnvalue_6))

(fn StageEffectsUpdate
  (params (Emitter :type Emitter%20Object%20Reference) (Setup%20Struc :type Papaya%20Effects%20Concert%20Stage%20Effects%20Structure))
  (bind (_spawnrateparametername_137_d0a1e18040781fa01eda27936d391489 _spawnrate_134_423bf929458961569240d8b988272a90 _active_133_f02dd95f4b00431fc3c56a8d62335ac7) (Utilities|Struct|BreakPapayaEffectsConcertStageEffects Setup%20Struc))
  (bind _particlesystemcomponent (Variables|/Script/CascadeRuntime.Emitter|GetParticleSystemComponent Emitter))
  (CallFunction|SetFloatParameterOnFXSystem _particlesystemcomponent _spawnrateparametername_137_d0a1e18040781fa01eda27936d391489 _spawnrate_134_423bf929458961569240d8b988272a90)
  (CallFunction|ToggleActivateFXSystem _particlesystemcomponent _active_133_f02dd95f4b00431fc3c56a8d62335ac7))

(fn StageEffectsUpdate_Niagara
  (params (Niagara%20System :type Niagara%20Particle%20System%20Component%20Object%20Reference) (Setup%20Struc :type Papaya%20Effects%20Concert%20Stage%20Effects%20Structure))
  (bind (_spawnrateparametername_137_d0a1e18040781fa01eda27936d391489 _spawnrate_134_423bf929458961569240d8b988272a90 _active_133_f02dd95f4b00431fc3c56a8d62335ac7) (Utilities|Struct|BreakPapayaEffectsConcertStageEffects Setup%20Struc))
  (CallFunction|SetFloatParameterOnFXSystem Niagara%20System _spawnrateparametername_137_d0a1e18040781fa01eda27936d391489 _spawnrate_134_423bf929458961569240d8b988272a90)
  (CallFunction|ToggleActivateFXSystem Niagara%20System _active_133_f02dd95f4b00431fc3c56a8d62335ac7 true))

(fn TimelineSetup
  (params (Fade%20Time :type Float%20%28double-precision%29) (In%20/%20Out :type Float%20%28double-precision%29) (Timeline :type Timeline%20Component%20Object%20Reference) (Control :type Float%20%28double-precision%29))
  (if (Components|Timeline|IsPlaying Timeline)
    (Components|Timeline|Stop Timeline)
    (Components|Timeline|SetNewTime Timeline (Variables|GetControl))
    (else
      (Components|Timeline|SetNewTime Timeline In%20/%20Out)))
  (Components|Timeline|SetPlayRate Timeline (/ 1.0 (select (<= Fade%20Time 0.0) 0.1 Fade%20Time)))
  (return))

(fn ToggleActivateFXSystem
  (params (FX%20System :type FXSystem%20Component%20Object%20Reference) (Activate :type Boolean) (Reset :type Boolean))
  (Utilities|IsValid FX%20System
    (:Is%20Valid
      (if Activate
        (Components|Activation|Activate FX%20System Reset)
        (else
          (Components|Activation|Deactivate FX%20System))))
    (:Is%20Not%20Valid
      (Development|PrintString "FX System Toggled is INVALID"))))

(fn ToggleActorHiddeninGame
  (params (Bool :type Boolean) (Actor :type Actor%20Object%20Reference))
  (Utilities|IsValid Actor
    (:Is%20Valid
      (Rendering|SetActorHiddenInGame Actor (not Bool)))))

(fn UpdateEverything
  (params (Updated%20Look :type Papaya%20Effects%20Whole%20Stage%20Structure))
  (bind _updated_look (Variables|GetUpdated%20Look))
  (bind (_dancer4active_98_a1d4de5a4b76aaf6cfcb788d0b26f140 _dancer4mesh_107_d1e758f94f1ca133fc8df691ba021714 _dancer4emote_116_90e8dad643bce850de4461846859cee2 _dancer4scale_125_a6ec6e2b451bb8e4156f21959eb2294f) (Utilities|Struct|BreakPapayaEffectsWholeStage _updated_look))
  (bind _updated_look_2 (Variables|GetUpdated%20Look))
  (bind (_dancer1active_95_6dfd15e6465618e28ab1c78908165489 _dancer1mesh_101_360acfcb475a75b2447b6ba301521168 _dancer1emote_110_d827f6f147c61d67f70868a48b529b1d _dancer1scale_119_8354278c4fa21b54f742848c65c910fa) (Utilities|Struct|BreakPapayaEffectsWholeStage _updated_look_2))
  (bind _updated_look_3 (Variables|GetUpdated%20Look))
  (bind (_basecolorparameter_61_324470b84d229bec7cb5da8d3c189768 _basecolormix_62_19310b22464aaef6226eecb5466b2dcd _intermissioncolorparameter_63_25cf8b9e41f1173509a166b60fb28e53 _intermissioncolormix_64_46267300430e61a8195be0bc4721c152 _desaturation_65_ef1eecf7416ef67d50d946a929c715cb) (Utilities|Struct|BreakPapayaEffectsWholeStage _updated_look_3))
  (bind _updated_look_4 (Variables|GetUpdated%20Look))
  (bind (_screenshardsactive_181_071ed8cc43e9e68afb36f4a294e3b46f _screenshardsspawnrate_184_b0c0c90c4011e58e7694b0963f83ff2b) (Utilities|Struct|BreakPapayaEffectsWholeStage _updated_look_4))
  (bind _updated_look_5 (Variables|GetUpdated%20Look))
  (bind (_screencurve1opacity_43_6fed178e41d2f967183af191b4c4cfb2 _screencurve2opacity_45_cdf37c8049613b0466c2afb490ed8710 _screencurve3opacity_48_c1c47a2645118eb225ec9993443b531f _screencurve4opacity_49_f0d407ea44298e32d51fd4b93a8a6183 _entryglowstripsr_53_5a5e21a7404351159e6027bab1a3410c _entryglowstripcenter_54_d8d5ba1b489f04f74e383abf6f1b130c _entryglowstripsl_55_04fd032346d7aab0c680529a0d042028) (Utilities|Struct|BreakPapayaEffectsWholeStage _updated_look_5))
  (bind _updated_look_6 (Variables|GetUpdated%20Look))
  (bind (_dancer2active_96_6724fbd3449b6e957b163391cbf467a2 _dancer2mesh_105_4ea9c4474c9b2e1a7e9deb8101cfc6c8 _dancer2emote_114_cde27c83407994c650ba5fab42ddee7c _dancer2scale_123_54fedeb943caf2e6eb7648aaad2fce53) (Utilities|Struct|BreakPapayaEffectsWholeStage _updated_look_6))
  (bind _updated_look_7 (Variables|GetUpdated%20Look))
  (bind (_laser1_77_3f49ea234e15b73e271578b1588659ee _laser2_79_980e0d24450138df30012b9cba949362 _laser3_84_7277631d40131c003cf75e9466b4fad8 _laser4_85_de32c190406aa8d330ec958ddba1881b _laser5_86_e9fc1a2f4fb8dfdef01f4187b7cbbcf8 _laser6_87_f4523b8f4009b3c4279f97bf9a24a424) (Utilities|Struct|BreakPapayaEffectsWholeStage _updated_look_7))
  (bind _updated_look_8 (Variables|GetUpdated%20Look))
  (bind (_concertrainactive_147_811702dd41f31054605f9fbc3c41b7ef _tvmode_150_74f777a04fc16800cd4412acddec3372 _spritedomeactive_158_6b055c7b4929b5a68d68029bfc0d358a _kaleidoscopeactive_157_bdc456e644a5ee8ad29167ac678c8550 _cartoonshademode_156_3307de674f3eae3396c019ab7d14fb5d _playerbubblesactive_162_f1466d9c47e049e391b489902a2de124 _playersparksactive_163_2f8bbda241ebbd0f7eed14836702b8b8 _playerheartsactive_164_51e1c12c44bca8376e73fa992afad647 _teaselightsactive_170_da3c9a194ec9ce11dc479fbf15bd532d _playerbouncyactive_171_beea3f924450e7d2207757a5881e2bf1 _playerfirejumpactive_172_e50d66744311a21091253da61c94b417 _plexusactive_173_0d38777e4b865953825b6487cedaf18e _holoplanesactive_174_598e0097443e58c27e691aa5f9568268 _globevisible_177_5ab8450644abe829281a5a893231d126) (Utilities|Struct|BreakPapayaEffectsWholeStage _updated_look_8))
  (bind _updated_look_9 (Variables|GetUpdated%20Look))
  (bind (_confettiactive_127_e473be284b69f2f4c37676800a0ccff7 _confettispawnrate_130_c455136749ec366a3516c59872159829 _discolightsactive_134_af0ee666407926c4df47a69e3dfdf54a _fireworksactive_142_b559e13348d5a46418b6a1b0bf48e941 _fireworksspawnrate_143_41cfbd1744d1e86eeabc838bf40cdaaf _sparksrainactive_144_2362cc9f407371f9d097aca5c9cb9633 _sparksrainspawnrate_145_3959c175487b5eedd402ef81d3dcf4c8) (Utilities|Struct|BreakPapayaEffectsWholeStage _updated_look_9))
  (bind _updated_look_10 (Variables|GetUpdated%20Look))
  (bind (_dancer3active_97_8f94ebd6400433166fb424bf6e8140a1 _dancer3mesh_106_ad9a2a0a4e84ae07613696b3a1ec0cd6 _dancer3emote_115_88f810104698c8fb22b0f8b4c5f12c16 _dancer3scale_124_8e9f25624319206d43699abdedb5bde1) (Utilities|Struct|BreakPapayaEffectsWholeStage _updated_look_10))
  (bind _updated_look_11 (Variables|GetUpdated%20Look))
  (bind (_srtetrascreen_74_d8f14d384ac97da90cc4e5aec7c42f50 _sltetrascreen_75_fd7245e9431e0affc67a879c864d3f49) (Utilities|Struct|BreakPapayaEffectsWholeStage _updated_look_11))
  (Utilities|FlowControl|Sequence
    (:then_0
      (Variables|StageSetup|Lasers|SetLaser1 _laser1_77_3f49ea234e15b73e271578b1588659ee)
      (Variables|StageSetup|Lasers|SetLaser2 _laser2_79_980e0d24450138df30012b9cba949362)
      (Variables|StageSetup|Lasers|SetLaser3 _laser3_84_7277631d40131c003cf75e9466b4fad8)
      (Variables|StageSetup|Lasers|SetLaser4 _laser4_85_de32c190406aa8d330ec958ddba1881b)
      (Variables|StageSetup|Lasers|SetLaser5 _laser5_86_e9fc1a2f4fb8dfdef01f4187b7cbbcf8)
      (Variables|StageSetup|Lasers|SetLaser6 _laser6_87_f4523b8f4009b3c4279f97bf9a24a424))
    (:then_1
      (Variables|StageSetup|Confetti|SetConfettiActive _confettiactive_127_e473be284b69f2f4c37676800a0ccff7)
      (Variables|StageSetup|Confetti|SetConfettiSpawnRate _confettispawnrate_130_c455136749ec366a3516c59872159829)
      (Variables|StageSetup|DiscoLights|SetDiscoLightsActive _discolightsactive_134_af0ee666407926c4df47a69e3dfdf54a)
      (Variables|StageSetup|Fireworks|SetFireworksActive _fireworksactive_142_b559e13348d5a46418b6a1b0bf48e941)
      (Variables|StageSetup|Fireworks|SetFireworksSpawnRate _fireworksspawnrate_143_41cfbd1744d1e86eeabc838bf40cdaaf)
      (Variables|StageSetup|SparksRain|SetSparksRainActive _sparksrainactive_144_2362cc9f407371f9d097aca5c9cb9633)
      (Variables|StageSetup|SparksRain|SetSparksRainSpawnRate _sparksrainspawnrate_145_3959c175487b5eedd402ef81d3dcf4c8))
    (:then_2
      (Variables|StageSetup|SetConcert%20Lighting%20Activate (Utilities|Struct|BreakPapayaEffectsWholeStage (Variables|GetUpdated%20Look))))
    (:then_3
      (Variables|Default|SetScreenShards _screenshardsactive_181_071ed8cc43e9e68afb36f4a294e3b46f)
      (Variables|Default|SetScreenShardsRate _screenshardsspawnrate_184_b0c0c90c4011e58e7694b0963f83ff2b))
    (:then_4
      (Variables|StageSetup|SetScreen%20Haze%20Opacity (Utilities|Struct|BreakPapayaEffectsWholeStage (Variables|GetUpdated%20Look))))
    (:then_5
      (Variables|StageSetup|SetScreen%20Nexus%20Activate (Utilities|Struct|BreakPapayaEffectsWholeStage (Variables|GetUpdated%20Look))))
    (:then_6
      (Variables|StageSetup|SetScreen%20Haze%20Opacity (Utilities|Struct|BreakPapayaEffectsWholeStage (Variables|GetUpdated%20Look))))
    (:then_7
      (Variables|StageSetup|SetScreen%20Side%20Equalizers%20Activate (Utilities|Struct|BreakPapayaEffectsWholeStage (Variables|GetUpdated%20Look))))
    (:then_8
      (Variables|StageSetup|GlowStrips|SetScreen%20Curve%201%20Alpha _screencurve1opacity_43_6fed178e41d2f967183af191b4c4cfb2)
      (Variables|StageSetup|GlowStrips|SetScreen%20Curve%202%20Alpha _screencurve2opacity_45_cdf37c8049613b0466c2afb490ed8710)
      (Variables|StageSetup|GlowStrips|SetScreen%20Curve%203%20Alpha _screencurve3opacity_48_c1c47a2645118eb225ec9993443b531f)
      (Variables|StageSetup|GlowStrips|SetScreen%20Curve%204%20Alpha _screencurve4opacity_49_f0d407ea44298e32d51fd4b93a8a6183)
      (Variables|StageSetup|GlowStrips|SetSR%20Entry%20GlowShape _entryglowstripsr_53_5a5e21a7404351159e6027bab1a3410c)
      (Variables|StageSetup|GlowStrips|SetCenter%20Entry%20GlowShape _entryglowstripcenter_54_d8d5ba1b489f04f74e383abf6f1b130c)
      (Variables|StageSetup|GlowStrips|SetSL%20Entry%20GlowShape _entryglowstripsl_55_04fd032346d7aab0c680529a0d042028))
    (:then_9
      (Variables|StageSetup|Color|SetBase%20ColorParameter _basecolorparameter_61_324470b84d229bec7cb5da8d3c189768)
      (Variables|StageSetup|Color|SetBase%20Color%20Mix _basecolormix_62_19310b22464aaef6226eecb5466b2dcd)
      (Variables|StageSetup|Color|SetIntermission%20Color%20Parameter _intermissioncolorparameter_63_25cf8b9e41f1173509a166b60fb28e53)
      (Variables|StageSetup|Color|SetIntermission%20Color%20Mix _intermissioncolormix_64_46267300430e61a8195be0bc4721c152)
      (Variables|StageSetup|Color|SetDesaturation _desaturation_65_ef1eecf7416ef67d50d946a929c715cb))
    (:then_10
      (Variables|StageSetup|SetSR%20TetraScreen _srtetrascreen_74_d8f14d384ac97da90cc4e5aec7c42f50)
      (Variables|StageSetup|SetSLTetraScreen _sltetrascreen_75_fd7245e9431e0affc67a879c864d3f49))
    (:then_11
      (Variables|StageSetup|SetDisco%20Ball (Utilities|Struct|BreakPapayaEffectsWholeStage (Variables|GetUpdated%20Look))))
    (:then_12
      (Variables|StageSetup|Dancer1|SetDancer1Active _dancer1active_95_6dfd15e6465618e28ab1c78908165489)
      (Variables|StageSetup|Dancer1|SetDancer1_Mesh _dancer1mesh_101_360acfcb475a75b2447b6ba301521168)
      (Variables|StageSetup|Dancer1|SetDancer1_Emote _dancer1emote_110_d827f6f147c61d67f70868a48b529b1d)
      (Variables|StageSetup|Dancer1|SetDancer1_Scale _dancer1scale_119_8354278c4fa21b54f742848c65c910fa))
    (:then_13
      (Variables|StageSetup|Dancer2|SetDancer2Active _dancer2active_96_6724fbd3449b6e957b163391cbf467a2)
      (Variables|StageSetup|Dancer2|SetDancer2_Mesh _dancer2mesh_105_4ea9c4474c9b2e1a7e9deb8101cfc6c8)
      (Variables|StageSetup|Dancer2|SetDancer2_Emote _dancer2emote_114_cde27c83407994c650ba5fab42ddee7c)
      (Variables|StageSetup|Dancer2|SetDancer2_Scale _dancer2scale_123_54fedeb943caf2e6eb7648aaad2fce53))
    (:then_14
      (Variables|StageSetup|Dancer3|SetDancer3Active _dancer3active_97_8f94ebd6400433166fb424bf6e8140a1)
      (Variables|StageSetup|Dancer3|SetDancer3_Mesh _dancer3mesh_106_ad9a2a0a4e84ae07613696b3a1ec0cd6)
      (Variables|StageSetup|Dancer3|SetDancer3_Emote _dancer3emote_115_88f810104698c8fb22b0f8b4c5f12c16)
      (Variables|StageSetup|Dancer3|SetDancer3_Scale _dancer3scale_124_8e9f25624319206d43699abdedb5bde1))
    (:then_15
      (Variables|StageSetup|Dancer4|SetDancer4Active _dancer4active_98_a1d4de5a4b76aaf6cfcb788d0b26f140)
      (Variables|StageSetup|Dancer4|SetDancer4_Mesh _dancer4mesh_107_d1e758f94f1ca133fc8df691ba021714)
      (Variables|StageSetup|Dancer4|SetDancer4_Emote _dancer4emote_116_90e8dad643bce850de4461846859cee2)
      (Variables|StageSetup|Dancer4|SetDancer4_Scale _dancer4scale_125_a6ec6e2b451bb8e4156f21959eb2294f))
    (:then_16
      (Variables|StageSetup|SetActivate%20Concert%20Rain _concertrainactive_147_811702dd41f31054605f9fbc3c41b7ef)
      (Variables|StageSetup|SetTVScreenMode _tvmode_150_74f777a04fc16800cd4412acddec3372)
      (Variables|StageSetup|SetSpriteDomeActive _spritedomeactive_158_6b055c7b4929b5a68d68029bfc0d358a)
      (Variables|StageSetup|SetKaleidoscopeValue (select _kaleidoscopeactive_157_bdc456e644a5ee8ad29167ac678c8550 1.0 0.0))
      (Variables|StageSetup|SetCartoonOutlinePreset _cartoonshademode_156_3307de674f3eae3396c019ab7d14fb5d)
      (Variables|StageSetup|SetActivatePlayerBubbles _playerbubblesactive_162_f1466d9c47e049e391b489902a2de124)
      (Variables|StageSetup|SetActivatePlayerSparkles _playersparksactive_163_2f8bbda241ebbd0f7eed14836702b8b8)
      (Variables|StageSetup|SetActivatePlayerHearts _playerheartsactive_164_51e1c12c44bca8376e73fa992afad647)
      (Variables|StageSetup|SetActivateSeqMovingLights _teaselightsactive_170_da3c9a194ec9ce11dc479fbf15bd532d)
      (Variables|StageSetup|SetActivatePlayerBouncy _playerbouncyactive_171_beea3f924450e7d2207757a5881e2bf1)
      (Variables|StageSetup|SetActivate%20Player%20Fire%20Jump _playerfirejumpactive_172_e50d66744311a21091253da61c94b417)
      (Variables|StageSetup|SetActivate%20Plexus _plexusactive_173_0d38777e4b865953825b6487cedaf18e)
      (Variables|StageSetup|SetHologramPlanesAlpha (select _holoplanesactive_174_598e0097443e58c27e691aa5f9568268 1.0 0.0))
      (Variables|StageSetup|SetWorldGlobeVisible _globevisible_177_5ab8450644abe829281a5a893231d126))))

(fn Update_MainScreenPhysProps
  (params (ScreenPhysAlpha :type Float%20%28double-precision%29))
  (bind _screen_actor (Variables|StageManagementDONOTEDIT!!|GetScreen%20Actor))
  (bind _output (Utilities|ResolveSoftReference _screen_actor))
  (bind _rootcomponent (Variables|/Script/Engine.Actor|GetRootComponent _output))
  (bind (_position _rotation _scale) (CallFunction|CreateScreenPhysData _rootcomponent))
  (bind _floating_large_shards (Variables|StageManagementDONOTEDIT!!|GetFloating%20Large%20Shards))
  (bind _output_2 (Utilities|ResolveSoftReference _floating_large_shards))
  (bind _rootcomponent_2 (Variables|/Script/Engine.Actor|GetRootComponent _output_2))
  (bind (_position_2 _rotation_2 _scale_2) (CallFunction|CreateScreenPhysData _rootcomponent_2 1))
  (bind _screen_nexusrays (Variables|StageManagementDONOTEDIT!!|GetScreen%20NexusRays))
  (bind _rootcomponent_3 (Variables|/Script/Engine.Actor|GetRootComponent _screen_nexusrays))
  (bind (_position_3 _rotation_3 _scale_3) (CallFunction|CreateScreenPhysData _rootcomponent_3 2))
  (Variables|Default|SetMainScreenPhysAlpha ScreenPhysAlpha)
  (Utilities|IsValid _rootcomponent
    (:Is%20Valid
      (Transformation|SetWorldLocation _rootcomponent _position false false)
      (Transformation|SetWorldRotation _rootcomponent _rotation false false)
      (Transformation|SetWorldScale3D _rootcomponent _scale)
      (if (Utilities|IsValidSoftObjectReference (Variables|StageManagementDONOTEDIT!!|GetFloating%20Large%20Shards))
        (Utilities|IsValid _rootcomponent_2
          (:Is%20Valid
            (Transformation|SetActorLocation _output_2 _position_2 false false)
            (Transformation|SetWorldRotation _rootcomponent_2 _rotation_2 false false)
            (Transformation|SetActorScale3D _output_2 _scale_2)
            (Utilities|IsValid _screen_nexusrays
              (:Is%20Valid
                (Transformation|SetActorLocation _screen_nexusrays _position_3 false false)
                (Transformation|SetWorldRotation _rootcomponent_3 _rotation_3 false false)
                (Transformation|SetActorScale3D _screen_nexusrays _scale_3)))))))))

(fn ConstructionScript)
