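; Blueprint stage manager gadget for timed events in the Papaya gameplay module, acting as a thin wrapper over native functions.
; Custom events to set up for a named screen, shut down, reset, and add a timed event each forward straight to their native counterpart.
; An added event carries a name, an integer time and a string of parameters.
; BeginPlay, overlap, tick and the construction script are empty.
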(event EventBeginPlay)

(event Collision|EventActorBeginOverlap
  (params (OtherActor :type Actor%20Object%20Reference)))

(event EventTick
  (params (DeltaSeconds :type Float%20%28single-precision%29)))

(event Custom|SetupTimedEvents
  (params (Screen :type String))
  (CallFunction|SetupTimedEventsNative Screen))

(event Custom|ShutdownTimedEvents
  (CallFunction|ShutdownTimedEventsNative))

(event Custom|ResetTimedEvents
  (CallFunction|ResetTimedEventsNative))

(event Custom|AddTimedEvent
  (params (EventName :type String) (Time :type Integer64) (Params :type String))
  (CallFunction|AddTimedEventNative EventName Time Params))

(fn ConstructionScript)
